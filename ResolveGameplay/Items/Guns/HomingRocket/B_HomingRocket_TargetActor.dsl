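; Homing rocket target marker actor that shows a decal and Niagara effect where the rocket is aimed.
; BeginPlay waits a tick, hands RocketObj to Niagara, and traces along the instigator-to-marker line to align with the surface or hide.
; On a hit it enables post-process and creates the decal's dynamic material.
; A cosmetic-only timer shrinks the decal as the rocket nears, updating its distance parameter by cached index.
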(event EventBeginPlay
  (Utilities|FlowControl|DelayUntilNextTick)
  (Networking|CanExecuteCosmeticEvents
    (:True
      (bind _rocketobj (Variables|Default|GetRocketObj)
        (:then
          (CallFunction|CalculateDistancefromRocketasPercentage _rocketobj (Variables|Default|GetMax%20Range%20For%20Material%20Adjustments))
          (CallFunction|InitializeDistanceParameter))
        (:else
          _)) ; May not have replicated to client yet,\nbut go ahead and try to initialize the value.
      (bind _returnvalue (Utilities|Time|SetTimerbyEvent (EventDispatchers|CreateEvent :!event Custom|UpdateDecal) 0.03 true))
      (Variables|Default|SetTimer _returnvalue))
    (:False
      _))
  (Niagara|SetNiagaraVariable(Object) (Variables|Default|GetNS_HomingRocket_Marker) "RocketObj" (Variables|Default|GetRocketObj))
  (bind _asfort_pawn (Utilities|Casting|CastToFortPawn (Game|GetInstigator))
    (:then
      ;[1 1 1] Do a trace upward and downward from the actors location
      (bind _returnvalue_2 (CallFunction|CalculateNormalizedDirectiontoInstigatorPawn _asfort_pawn))
      (bind (_outhit _returnvalue_3) (Collision|LineTraceForObjects :Start (+ (* _returnvalue_2 -5e+02) (Transformation|GetWorldLocation (Variables|Default|GetDefaultSceneRoot))) :End (+ (* _returnvalue_2 5e+02) (Transformation|GetWorldLocation (Variables|Default|GetDefaultSceneRoot))) :ObjectTypes (Variables|Default|GetTraceObjectTypes) :bTraceComplex false :DrawDebugType "None" :bIgnoreSelf true :TraceColor "(R=1.000000,G=0.000000,B=0.000000,A=1.000000)" :TraceHitColor "(R=0.000000,G=1.000000,B=0.000000,A=1.000000)" :DrawTime 1e+01))
      ;[1 1 1] If the trace hits, rotate the actor to match the impact normal, If no hits, hide the decal (Fixed spawning in middair
      (if _returnvalue_3
        (Transformation|SetActorRotation (Math|Rotator|MakeRotfromZ (.ImpactNormal (Collision|BreakHitResult _outhit))))
        (Variables|/Script/Engine.PostProcessComponent|SetbEnabled true (Variables|Default|GetPostProcess))
        (bind _returnvalue_4 (Rendering|Material|CreateDynamicMaterialInstance (Variables|/Script/Engine.DecalComponent|GetDecalMaterial (Variables|Default|GetDecal)) "None" "None" :!class /Script/Engine.KismetMaterialLibrary))
        (Variables|Default|SetDecalMID _returnvalue_4)
        (Rendering|Components|Decal|SetDecalMaterial (Variables|Default|GetDecal) (Variables|Default|GetDecalMID))
        (else
          (Development|SetHiddeninGame (Variables|Default|GetDefaultSceneRoot) true true))))
    (:CastFailed
      (Development|PrintString (Utilities|GetDisplayName _asfort_pawn) true true "(R=1.000000,G=0.000000,B=0.025866,A=1.000000)" 5.0))))

(event EventDestroyed
  (Utilities|Time|ClearandInvalidateTimerbyHandle (Variables|Default|GetTimer)))

(event Custom|UpdateDecal
  (Utilities|IsValid (Variables|Default|GetDecalMID)
    (:Is%20Valid
      (bind _rocketobj (Variables|Default|GetRocketObj)
        (:then
          (bind _targetpercentdistancefromrocket (CallFunction|CalculateDistancefromRocketasPercentage _rocketobj))
          (if (!= (Variables|Default|GetTarget%20Percent%20Distance%20From%20Rocket) (Variables|Default|GetPrevious%20Target%20Percent%20Distance%20From%20Rocket))
            (Transformation|SetRelativeScale3D (Variables|Default|GetDecal) (Math|Vector|Lerp(Vector) "0, 0, 0" "5.000000,5.000000,5.000000" _targetpercentdistancefromrocket))
            (if (Variables|Default|GetbInitializedDistanceParameterIndex)
              (Rendering|Material|SetScalarParameterbyIndex (Variables|Default|GetDecalMID) (Variables|Default|GetMaterialDistancePercentageParameterIndex) (Variables|Default|GetTarget%20Percent%20Distance%20From%20Rocket)) ; More efficient way to set index.
              (else
                (CallFunction|InitializeDistanceParameter)))))
        (:else)))))

(fn CalculateDistanceFromRocketAsPercentage
  (params (Valid%20Rocket%20Actor :type Scene%20Component%20Object%20Reference) (MaxRange :type Float%20%28double-precision%29 :default 3.5e+03))
  (bind _distance (Class|MissionBlueprintFunctionLibrary|DistanceBetweenTwoVectors (Transformation|GetActorLocation) (Transformation|GetWorldLocation Valid%20Rocket%20Actor)))
  (Variables|Default|SetPrevious%20Target%20Percent%20Distance%20From%20Rocket (Variables|Default|GetTarget%20Percent%20Distance%20From%20Rocket)) ; Store previous value
  (bind _output_get (Variables|Default|SetTarget%20Percent%20Distance%20From%20Rocket (Math|Float|MapRangeClamped _distance 0.0 MaxRange 0.0 1.0))) ; Set New Value
  (return _output_get))

(fn CalculateNormalizedDirectionToInstigatorPawn
  (params (FortPawn :type Fort%20Pawn%20Object%20Reference))
  (return (Math|Vector|Normalize (- (Transformation|GetActorLocation) (Targeting|GetDefaultTargetingOrigin FortPawn)))))

(fn InitializeDistanceParameter
  (bind _decalmid (Variables|Default|GetDecalMID))
  (Utilities|IsValid _decalmid
    (:Is%20Valid
      (bind (_outparameterindex _returnvalue) (Rendering|Material|InitializeScalarParameterandGetIndex _decalmid (Utilities|Name|MakeLiteralName "PercentDistanceFromTarget") (Variables|Default|GetTarget%20Percent%20Distance%20From%20Rocket)))
      (Variables|Default|SetMaterialDistancePercentageParameterIndex _outparameterindex)
      (Variables|Default|SetbInitializedDistanceParameterIndex _returnvalue))))

(fn ConstructionScript)
