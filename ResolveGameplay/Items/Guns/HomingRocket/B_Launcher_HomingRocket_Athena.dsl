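; Homing rocket launcher weapon Blueprint for Fortnite Athena, handling rocket guidance, scope visuals and lock-on feedback.
; On projectile spawn, it passes the rocket the client-supplied lock-on target, or none, plus an aim point from a damage-start line trace.
; Aiming toggles custom-depth significance handling for the local viewer, while a timeline drives thermal scope material parameters.
; Lock-on states select one-shot and looping targeting sounds, and firing activates rear-muzzle smoke.
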(event EventOnSetTargeting
  (params (bNewIsTargeting :type Boolean))
  ;[0 1 0.1139] Scope Post Process Setup
  (Networking|CanExecuteCosmeticEvents
    (:True
      (bind _asfort_pawn (Utilities|Casting|CastToFortPawn (Game|GetInstigator)))
      (Utilities|IsValid (Pawn|GetLocalViewingPlayerController _asfort_pawn)
        (:Is%20Valid
          (|CallParentFunction bNewIsTargeting)
          (bind _asfort_pawn_2 (Utilities|Casting|CastToFortPawn (Game|GetInstigator)))
          (bind _asfort_player_controller_athena (Utilities|Casting|CastToFortPlayerControllerAthena (Pawn|GetLocalViewingPlayerController _asfort_pawn_2)))
          (Rendering|SetIgnoreSignifanceBasedCustomDepthRendering _asfort_player_controller_athena (Camera|GetIsTargeting) 0))))))

(event EventPlayScopePP
  ;[0 1 0.1139] Scope Post Process Setup
  (Utilities|FlowControl|Delay 0.1)
  (if (Camera|GetIsTargeting)
    (CallFunction|SetPostProcessParams 1.0)
    (Components|Timeline|SetPlayRate (Variables|B_Launcher_HomingRocket_Athena|GetScalePlayerThermal) (/ 1.0 (Variables|Anvil|Floats|GetBlendInTime)))
    (Rendering|Material|SetScalarParameterValue "/Game/Athena/Items/Weapons/Abilities/Thermal/MPC_ThermalGlobal.MPC_ThermalGlobal" "ThermalFadeout" 1.0 :!class /Script/Engine.KismetMaterialLibrary)
    (Utilities|FlowControl|DoOnce false
      (:Completed
        (bind (_direction _newtrack_0) (Timeline|ScalePlayerThermal PlayFromStart)
          (:Update
            (Rendering|Material|SetScalarParameterValue "/Game/Athena/Items/Weapons/Abilities/Thermal/MPC_ThermalGlobal.MPC_ThermalGlobal" "ThermalScale" _newtrack_0 :!class /Script/Engine.KismetMaterialLibrary))
          (:Finished))))))

(event EventBeginPlay
  ;[1 1 1] Establish Player Reference
  (|CallParentFunction)
  (bind _asfort_pawn (Utilities|Casting|CastToFortPawn (Game|GetInstigator)))
  (Variables|Default|SetPlayerRef _asfort_pawn))

(event FortWeapon|EventOnUnEquippedAndDetached
  ;[1 1 1] Kills sound when dropped
  (|CallParentFunction)
  (Components|DestroyComponent (Variables|Default|GetLockOnSoundCue)))

(event Custom|PassServerLockOnTarget
  (params (LockOnTarget :type Actor%20Object%20Reference))
  ;[1 1 1] Pass client target to server
  (Variables|Default|SetLockOnTarget LockOnTarget))

(event Custom|UpdateJavelinHomingData
  (params (SpawnedProjectile :type Object%20Reference))
  ;[0 0.03702 1] Logic Call to Lock On GA
  ;;[1 1 1] If target it valid, target them
  (bind _aimpointlocation (CallFunction|ExecuteTraceforLocation))
  (Utilities|IsValid (Variables|Default|GetLockOnTarget)
    (:Is%20Valid
      (bind _asb_prj_ranged_homing_rocket_athena (Utilities|Casting|CastToB_Prj_Ranged_HomingRocket_Athena SpawnedProjectile))
      (Class|BPrjRangedHomingRocketAthena|UpdateJavelinHomingData _asb_prj_ranged_homing_rocket_athena (Variables|Default|GetLockOnTarget) _aimpointlocation))
    (:Is%20Not%20Valid
      ;;[1 1 1] If not valid, target the aim point
      (bind _asb_prj_ranged_homing_rocket_athena_2 (Utilities|Casting|CastToB_Prj_Ranged_HomingRocket_Athena SpawnedProjectile))
      (Class|BPrjRangedHomingRocketAthena|UpdateJavelinHomingData _asb_prj_ranged_homing_rocket_athena_2 0 _aimpointlocation))))

(event FortWeapon|EventOnSpawnProjectile
  (params (SpawnProjectile :type Fort%20Projectile%20Base%20Object%20Reference))
  ;[0 0.03702 1] Logic Call to Lock On GA
  (CallFunction|UpdateJavelinHomingData SpawnProjectile))

(event Custom|PlayTargetingSounds
  (params (TargetingState :type ENUM_HomingRocket_LockOnStates%20Enum :default "NewEnumerator0"))
  (bind _targetingaudio (Variables|Default|GetTargetingAudio))
  ;[0.9979 1 0] Targeting Audio. Overengineered? Sure. But does it work? Maybe.
  (Audio|Components|Audio|Stop (Variables|Default|GetTargetingOneShot))
  (bind _returnvalue (Audio|SpawnSound2D (Utilities|Select :Index TargetingState :NewEnumerator0 (Variables|Anvil|Sounds|GetSearchingSound) :NewEnumerator1 (Variables|Anvil|Sounds|GetLockingOnSound) :NewEnumerator2 (Variables|Anvil|Sounds|GetLockedOnSound) :NewEnumerator5 (Variables|Anvil|Sounds|GetSearchingSound))))
  (Variables|Default|SetTargetingOneShot _returnvalue)
  (Audio|Components|Audio|Stop (Variables|Default|GetLockingOnLoop))
  (switch Utilities|FlowControl|Switch|SwitchonENUM_HomingRocket_LockOnStates TargetingState
    (:NewEnumerator0
      _)
    (:NewEnumerator1
      (bind _returnvalue_2 (Audio|SpawnSound2D "/ResolveGameplay/Items/Guns/HomingRocket/Sounds/Launcher_HomingRocket_ADS_TargetingLockingOn_Loop_Cue.Launcher_HomingRocket_ADS_TargetingLockingOn_Loop_Cue"))
      (Variables|Default|SetLockingOnLoop _returnvalue_2))
    (:NewEnumerator2
      (Variables|Anvil|Floats|SetTargetAudioFadeInTime 0.2)
      (Audio|Parameter|SetIntegerParameter _targetingaudio "TargetState" 2 :!class /Script/AudioExtensions.AudioParameterControllerInterface))
    (:NewEnumerator3
      (Audio|Components|Audio|Stop (Variables|Default|GetTargetingAudio))
      (Variables|Anvil|Floats|SetTargetAudioFadeInTime 1.0))
    (:NewEnumerator5
      _))
  (if (not (Audio|Components|Audio|IsPlaying _targetingaudio))
    (Audio|Components|Audio|SetSound _targetingaudio "/ResolveGameplay/Items/Guns/HomingRocket/Sounds/Launcher_HomingRocket_ADS_TargetingStates_Loop_Cue.Launcher_HomingRocket_ADS_TargetingStates_Loop_Cue")
    (Audio|Components|Audio|FadeIn _targetingaudio (Variables|Anvil|Floats|GetTargetAudioFadeInTime) 1.0 0.0 "Linear" :!class /Script/Engine.AudioComponent))
  (Audio|Parameter|SetIntegerParameter _targetingaudio "TargetState" 1 :!class /Script/AudioExtensions.AudioParameterControllerInterface))

(event EventOnPlayWeaponFireFX
  (params (bPersistentFire :type Boolean) (bSecondaryFire :type Boolean))
  ;[1 1 1] Smoke on Fire FX
  (|CallParentFunction bPersistentFire bSecondaryFire)
  (Components|Activation|Activate (Variables|Default|GetNS_Javelin_RearMuzzle) false))

(fn ExecuteTraceForLocation
  (bind (_aimdir _returnvalue) (FortWeapon|GetDamageStartLocation -1))
  (FortWeapon|UpdateDamageStartLocation)
  (bind _outhit (Collision|LineTraceByChannel :Start _returnvalue :End (+ _returnvalue (* _aimdir (ScalableFloat|GetValueatLevel (Variables|Default|GetRowTraceDistance)))) :TraceChannel "TraceTypeQuery6" :bTraceComplex false :DrawDebugType (select (Variables|Anvil|Booleans|GetDebug) "ForOneFrame" "None") :bIgnoreSelf true :TraceColor "(R=1.000000,G=0.000000,B=0.000000,A=1.000000)" :TraceHitColor "(R=0.000000,G=1.000000,B=0.000000,A=1.000000)" :DrawTime 5.0))
  (return (.ImpactPoint (Collision|BreakHitResult _outhit))))

(fn ConstructionScript
  (|CallParentFunction))
