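; Fortnite homing rocket projectile that locks onto a vehicle or other target, or flies to an aim point with a marker decal, then explodes.
; BeginPlay timers stagger homing, trail, turn-speed lerp, lifespan and proximity checks; an upward trace enables homing early under obstructions.
; Lock-on applies warning GameplayEffects and tracks vehicle passengers.
; Exploding runs surface or water cues and pushes nearby objects with an impulse.
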(event EventBeginPlay
  (bind _missilemesh (Variables|Default|GetMissileMesh))
  ;[0 0.0427 1] Rocket Behavior
  ;;[1 1 1] Begin Play (Getting a bunch of timers ready to fire and update)
  (|CallParentFunction)
  ;;;[1 1 1] Logic added to have prj spawn from Muzzle of launcher
  (Networking|SwitchHasAuthority
    (:Authority
      _)
    (:Remote
      (bind _asfort_player_pawn (Utilities|Casting|CastToFortPlayerPawn (Game|GetInstigator (Game|GetInstigator))))
      (Variables|Default|SetMuzzleLoc (Transformation|GetSocketLocation (Variables|/Script/FortniteGame.FortWeapon|GetWeaponMesh (Variables|/Script/FortniteGame.FortPawn|GetCurrentWeapon _asfort_player_pawn)) "Muzzle"))
      (Networking|ForceNetUpdate)))
  (Variables|Anvil|Floats|SetTimeAtStartofLife (Fortnite|Sync|Time|GetServerWorldTimeSeconds))
  (Variables|/Script/Engine.ProjectileMovementComponent|SetbRotationFollowsVelocity false (Projectile|GetFortProjectileMovementComponent :!class /Script/FortniteGame.FortProjectileMovementInterface))
  (if (== (ScalableFloat|GetValueatLevel (Variables|Anvil|ScalableFloats|GetRowPlayerCanStepUpOn)) 1.0)
    (Variables|/Script/Engine.PrimitiveComponent|SetCanCharacterStepUpOn "ECB_Yes" _missilemesh)
    (else
      (Variables|/Script/Engine.PrimitiveComponent|SetCanCharacterStepUpOn "ECB_No" _missilemesh)))
  (Utilities|Time|SetTimerbyEvent (EventDispatchers|CreateEvent :!event Custom|EnableHoming) (Variables|Anvil|Floats|GetDelayBeforeHoming))
  (Utilities|Time|SetTimerbyEvent (EventDispatchers|CreateEvent :!event Custom|EnableRocketTrail) (Variables|Anvil|Floats|GetDelayBeforeTrail))
  (Utilities|Time|SetTimerbyEvent (EventDispatchers|CreateEvent :!event Custom|EnableVelocityFollowRotation) (Variables|Anvil|Floats|GetDelayBeforeRotation))
  (bind _returnvalue (Utilities|Time|SetTimerbyEvent (EventDispatchers|CreateEvent :!event Custom|UpdateTurnSpeed) (Variables|Anvil|Floats|GetTurnUpdateTickRate) true))
  (Variables|Default|SetTurnTimerHandle _returnvalue)
  (Utilities|Time|SetTimerbyEvent (EventDispatchers|CreateEvent :!event Custom|EndOfLifeExplosion) (ScalableFloat|GetValueatLevel (Variables|Anvil|ScalableFloats|GetRowLifeSpan)) true)
  (Utilities|Time|SetTimerbyEvent (EventDispatchers|CreateEvent :!event Custom|ActivateTargetMarkerDecal) (Variables|Anvil|Floats|GetDelayBeforeActivateTargetMarker) true)
  (bind _returnvalue_2 (Utilities|Time|SetTimerbyEvent (EventDispatchers|CreateEvent :!event Custom|BankingSoundCheck) 0.1 true false 2.0))
  (Variables|Audio|SetTurnSoundTimerHandle _returnvalue_2)
  (Utilities|Time|SetTimerbyEvent (EventDispatchers|CreateEvent :!event Custom|UpdateCanExplodeInWater) (Variables|Default|GetCanExplodeinWaterDelay) false false 1.0)
  (Utilities|FlowControl|Delay (ScalableFloat|GetValueatLevel (Variables|Anvil|ScalableFloats|GetRowObjectAboveCheck)))
  (Utilities|Time|SetTimerbyEvent (EventDispatchers|CreateEvent :!event Custom|ObjectAboveCheck) (Variables|Anvil|Floats|GetObjectAboveCheckTick) true)
  (Utilities|FlowControl|Delay (ScalableFloat|GetValueatLevel (Variables|Anvil|ScalableFloats|GetRowDistanceToTargetCheck)))
  (Networking|SwitchHasAuthority
    (:Authority
      (bind _returnvalue_3 (Utilities|Time|SetTimerbyEvent (EventDispatchers|CreateEvent :!event Custom|PlayIncomingWarningSound) (Variables|Anvil|Floats|GetDistanceCheckTickRate) true))
      (Variables|Default|SetDistanceCheckTimerHandle _returnvalue_3))
    (:Remote
      _))
  (Utilities|Time|SetTimerbyEvent (EventDispatchers|CreateEvent :!event Custom|InAirLocationReachedCheck) (Variables|Anvil|Floats|GetInAirLocationReachTickRate) true))

(event Custom|ObjectAboveCheck
  ;[0 0.0427 1] Rocket Behavior
  ;;[1 1 1] Check for an object above the rocket after a brief delay. If something is above it, enable homing sooner
  (bind (_outhit _returnvalue) (Collision|LineTraceByChannel :Start (Transformation|GetWorldLocation (Variables|Collision|GetCapsuleComponent)) :End (+ (Transformation|GetWorldLocation (Variables|Collision|GetCapsuleComponent)) (* (Transformation|GetUpVector (Variables|Collision|GetCapsuleComponent)) 2e+03)) :TraceChannel "TraceTypeQuery6" :bTraceComplex false :DrawDebugType (select (Variables|Default|GetDebug) "ForOneFrame" "None") :bIgnoreSelf true :TraceColor "(R=1.000000,G=0.000000,B=0.000000,A=1.000000)" :TraceHitColor "(R=0.000000,G=1.000000,B=0.000000,A=1.000000)" :DrawTime 5.0))
  (if _returnvalue
    (Development|PrintString "Something's Above ME!" (Variables|Default|GetDebug) (Variables|Default|GetDebug))
    (CallFunction|EnableHoming)))

(event Custom|EnableHoming
  ;[0 0.0427 1] Rocket Behavior
  ;;[1 1 1] Call to set the prj to lockon
  (Projectile|SetHomingStyle 0 0 "LockOn" false :!class /Script/FortniteGame.FortProjectileMovementInterface))

(event Custom|UpdateTurnSpeed
  (bind _rowtimetominturnspeed (Variables|Anvil|ScalableFloats|GetRowTimeToMinTurnSpeed))
  (bind _returnvalue (ScalableFloat|GetValueatLevel _rowtimetominturnspeed))
  (bind _timeatstartoflife (Variables|Anvil|Floats|GetTimeAtStartofLife))
  (bind _returnvalue_2 (Fortnite|Sync|Time|GetServerWorldTimeSeconds))
  (bind _returnvalue_3 (- _returnvalue_2 _timeatstartoflife))
  (bind _returnvalue_4 (Math|Float|SafeDivide _returnvalue_3 _returnvalue))
  ;[0 0.0427 1] Rocket Behavior
  ;;[1 1 1] Updating Rockets Turn Speed over time
  (if (Networking|IsServer)
    (if (<= _returnvalue_4 0)
      (Utilities|Time|ClearandInvalidateTimerbyHandle (Variables|Default|GetTurnTimerHandle))
      (else
        (Projectile|SetHomingTurnSpeedMax (Projectile|GetFortProjectileMovementComponent :!class /Script/FortniteGame.FortProjectileMovementInterface) (Math|Float|Lerp (ScalableFloat|GetValueatLevel (Variables|Anvil|ScalableFloats|GetRowInitialMaxTurnSpeed)) (ScalableFloat|GetValueatLevel (Variables|Anvil|ScalableFloats|GetRowReducedMaxTurnSpeed)) _returnvalue_4) :!class /Script/FortniteGame.FortProjectileMovementComponent)))))

(event Custom|ApplyImpulse
  (bind _debug (Variables|Default|GetDebug))
  (bind _marker_actor (Variables|Default|GetMarker%20Actor))
  ;[0 0.0427 1] Rocket Behavior
  ;;[1 1 1] Applying Impulse
  ;;;[1 1 1] Debug Prints
  (Development|PrintString "Killing Target!" _debug _debug "(R=1.000000,G=0.005116,B=0.000000,A=1.000000)")
  (Development|PrintString (Utilities|GetDisplayName _marker_actor) _debug _debug)
  ;;;[1 1 1] Destroy TargetActor
  (Actor|DestroyActor _marker_actor)
  (Networking|SwitchHasAuthority
    (:Authority
      (CallFunction|RemoveOneLockonStack)
      ;;;[1 1 1] Sphere Trace at impact and add all vehicles to array
      (bind _outhits (Collision|MultiSphereTraceForObjects (Transformation|GetActorLocation) (Transformation|GetActorLocation) 777.0 (Utilities|Array|MakeArray "ObjectTypeQuery5")))
      (bind _array_element (Utilities|Array|ForEachLoopwithBreak _outhits)
        (:LoopBody
          ;;;[1 1 1] Wake up the vehicle for impulse
          (Vehicle|WakeUp(Message) (.HitActor (Collision|BreakHitResult _array_element)))
          (bind (_returnvalue_x _returnvalue_y _returnvalue_z) (Math|Vector|BreakVector (* (Math|Vector|Normalize (- (Transformation|GetActorLocation (.HitActor (Collision|BreakHitResult _array_element))) (Transformation|GetActorLocation))) 7.5e+02)))
          (Audio|BroadcastAbilityImpactatLocation 0 (.HitActor (Collision|BreakHitResult _array_element)) (Math|Vector|MakeVector _returnvalue_x _returnvalue_y (Math|Float|Max(Float) (+ _returnvalue_z 5e+02) 1e+03)) 777.0))))))

(event Custom|ActivateTargetMarkerDecal
  (bind _marker_actor (Variables|Default|GetMarker%20Actor))
  ;[0 0.0427 1] Rocket Behavior
  ;;[1 1 1] Event Timers
  ;;;[1 1 1] Sets visbility of target actor to True
  (Utilities|IsValid _marker_actor
    (:Is%20Valid
      (Development|SetHiddeninGame (Variables|/ResolveGameplay/Items/Guns/HomingRocket/B_HomingRocket_TargetActor.B_HomingRocket_TargetActor_C|GetDecal _marker_actor) false false))))

(event Custom|EndOfLifeExplosion
  ;[0 0.0427 1] Rocket Behavior
  ;;[1 1 1] Event Timers
  ;;;[1 1 1] End of life span, do the explosion
  (Delivery|DoExplosion))

(event Custom|EnableRocketTrail
  (bind _ns_homingrocket_prj_trail (Variables|Default|GetNS_HomingRocket_Prj_Trail))
  ;[0 0.0427 1] Rocket Behavior
  ;;[1 1 1] Event Timers
  ;;;[1 1 1] Enables the rocket trail FX
  (if (not (Variables|Default|GetIsHidden?))
    (Components|Activation|Activate _ns_homingrocket_prj_trail true)
    (Effects|Components|ParticleSystem|SetFloatParameter _ns_homingrocket_prj_trail "Startboosting" 1.0)))

(event Custom|EnableVelocityFollowRotation
  ;[0 0.0427 1] Rocket Behavior
  ;;[1 1 1] Event Timers
  ;;;[1 1 1] Stores intial rotation and plays timeline
  (Variables|Default|SetInitialRotation (Transformation|GetWorldRotation (Variables|Default|GetMissileMesh)))
  (Components|Timeline|PlayfromStart (Variables|B_Prj_Ranged_HomingRocket_Athena|GetRotationFollowsVelocity)))

(event Custom|UpdateJavelinHomingData
  (params (LockOnTarget :type Actor%20Object%20Reference) (TargetLoc :type Vector))
  ;[0 0.0427 1] Rocket Behavior
  ;;[1 1 1] Rocket Homing Data Updates
  ;;;[1 1 1] Set Lock On and check if we actually have one
  (bind _output_get (Variables|Default|SetLock%20on%20Target LockOnTarget))
  (Utilities|IsValid _output_get
    (:Is%20Valid
      ;;;[1 1 1] If Target, spawn alarm sound and attach to target
      (GameplayEffects|ApplyGameplayEffectToTarget :self (Ability|GetAbilitySystemComponent LockOnTarget) :GameplayEffectClass (Variables|Default|GetGEMissileLock) :Target (Ability|GetAbilitySystemComponent LockOnTarget) :Level 0.0)
      ;;;[1 1 1] Feed in lock on homing data and apply widget GE to targets
      (Projectile|SetHomingData (Variables|Movement|GetProjectileMovementComponent) (Utilities|Struct|MakeProjectileHomingData "None" (ScalableFloat|GetValueatLevel (Variables|Anvil|ScalableFloats|GetRowInitialMinTurnSpeed)) (ScalableFloat|GetValueatLevel (Variables|Anvil|ScalableFloats|GetRowInitialMaxTurnSpeed)) 5.0 3.0 (Math|Random|RandomInteger 14) LockOnTarget 0 "0.000000,0.000000,0.000000" (CallFunction|GetTargetOffset LockOnTarget TargetLoc) "Relative" false true) :!class /Script/FortniteGame.FortProjectileMovementComponent)
      (Utilities|IsValid LockOnTarget
        (:Is%20Valid
          (CallFunction|AddWarningWidgettoTargets LockOnTarget))))
    (:Is%20Not%20Valid
      (Variables|Default|SetTarget%20Loc TargetLoc)
      ;;;[1 1 1] Set homing data to lock onto aim point
      (Utilities|FlowControl|Sequence
        (:then_0
          (Projectile|SetHomingData (Variables|Movement|GetProjectileMovementComponent) (Utilities|Struct|MakeProjectileHomingData "None" (ScalableFloat|GetValueatLevel (Variables|Anvil|ScalableFloats|GetRowInitialMinTurnSpeed)) (ScalableFloat|GetValueatLevel (Variables|Anvil|ScalableFloats|GetRowInitialMaxTurnSpeed)) 5.0 3.0 (Math|Random|RandomInteger 14) 0 0 TargetLoc "0.000000,0.000000,0.000000" "Auto" false true) :!class /Script/FortniteGame.FortProjectileMovementComponent))
        (:then_1
          (CallFunction|CESpawnMarkerActor TargetLoc))))))

(event Custom|InAirLocationReachedCheck
  (bind _capsulecomponent (Variables|Collision|GetCapsuleComponent))
  (bind _returnvalue (Transformation|GetWorldLocation _capsulecomponent))
  ;[0 0.0427 1] Rocket Behavior
  ;;[1 1 1] Check to see if we reached our location IF we don't have a lock on target (Fixed rocket behavior in air with no target)
  (Utilities|IsValid (Variables|Default|GetLock%20on%20Target)
    (:Is%20Not%20Valid
      (if (<= (Math|Vector|VectorLength (- _returnvalue (Variables|Default|GetTarget%20Loc))) 2.5e+02)
        (Delivery|DoExplosionatLocation _returnvalue)))))

(event EventOnStop
  (params (Hit :type Hit%20Result%20Structure%20%28by%20ref%29))
  (bind (_bblockinghit _binitialoverlap _time _distance _location _impactpoint _normal _impactnormal _physmat) (Collision|BreakHitResult Hit))
  (bind _missilemesh (Variables|Default|GetMissileMesh))
  (bind _returnvalue (Transformation|GetWorldLocation _missilemesh))
  ;[0 0.0427 1] Rocket Behavior
  ;;[1 1 1] Impact and Explosion Stuff
  (Utilities|IsValid _physmat
    (:Is%20Valid
      (Variables|Default|SetSurfaceType (Variables|/Script/PhysicsCore.PhysicalMaterial|GetSurfaceType _physmat)))
    (:Is%20Not%20Valid
      _))
  (if (not (Delivery|IsBeingKilled))
    (Delivery|DoExplosionatLocation _returnvalue)
    (Fortnite|ShowSoundIndicatorLocal :Location _returnvalue :MaxAudibleDistance 5e+03 :IndicatorType "Gunshot" :Instigator self :Tint "(R=0.000000,G=0.000000,B=0.000000,A=0.000000)" :AttenuationMaxRangeMultiplier 1.0 :AttenuationIntensityMultiplier 1.0)
    (GameplayEffects|RemoveActiveGameplayEffectbySourceEffect (Ability|GetAbilitySystemComponent (Variables|Default|GetLock%20on%20Target)) (Variables|Default|GetGEMissileLock) (Ability|GetAbilitySystemComponent (Variables|Default|GetLock%20on%20Target)) 1)
    (GameplayEffects|RemoveActiveGameplayEffectbySourceEffect (Ability|GetAbilitySystemComponent (Variables|Default|GetLock%20on%20Target)) (Variables|Default|GetGEIncomingMissile) (Ability|GetAbilitySystemComponent (Variables|Default|GetLock%20on%20Target)) 1)
    (CallFunction|UnbindPlayerEnteredVehicle)))

(event Custom|FuseTimerMax
  (bind _missilemesh (Variables|Default|GetMissileMesh))
  (bind _returnvalue (Transformation|GetWorldLocation _missilemesh))
  ;[0 0.0427 1] Rocket Behavior
  ;;[1 1 1] Impact and Explosion Stuff
  (Variables|Default|SetTimerMaxReached? true)
  (if (not (Delivery|IsBeingKilled))
    (Delivery|DoExplosionatLocation _returnvalue)
    (Fortnite|ShowSoundIndicatorLocal :Location _returnvalue :MaxAudibleDistance 5e+03 :IndicatorType "Gunshot" :Instigator self :Tint "(R=0.000000,G=0.000000,B=0.000000,A=0.000000)" :AttenuationMaxRangeMultiplier 1.0 :AttenuationIntensityMultiplier 1.0)
    (GameplayEffects|RemoveActiveGameplayEffectbySourceEffect (Ability|GetAbilitySystemComponent (Variables|Default|GetLock%20on%20Target)) (Variables|Default|GetGEMissileLock) (Ability|GetAbilitySystemComponent (Variables|Default|GetLock%20on%20Target)) 1)
    (GameplayEffects|RemoveActiveGameplayEffectbySourceEffect (Ability|GetAbilitySystemComponent (Variables|Default|GetLock%20on%20Target)) (Variables|Default|GetGEIncomingMissile) (Ability|GetAbilitySystemComponent (Variables|Default|GetLock%20on%20Target)) 1)
    (CallFunction|UnbindPlayerEnteredVehicle)))

(event EventDestroyed
  ;[0 0.0427 1] Rocket Behavior
  ;;[1 1 1] Impact and Explosion Stuff
  (Networking|CanExecuteCosmeticEvents
    (:True
      (Components|Activation|SetActive (Variables|Default|GetNS_HomingRocket_Prj_Trail) false false)
      (Development|SetHiddeninGame (Variables|Default|GetMissileMesh) true false)
      (Actor|DestroyActor (Variables|Default|GetMarker%20Actor)))))

(event EventOnExploded
  (params (HitActors :type Array%20of%20Actor%20Object%20References) (HitResults :type Array%20of%20Hit%20Result%20Structures))
  ;[0 0.0427 1] Rocket Behavior
  ;;[1 1 1] Impact and Explosion Stuff
  (|CallParentFunction HitActors HitResults)
  (Utilities|FlowControl|Sequence
    (:then_0
      (Networking|CanExecuteCosmeticEvents
        (:True
          (Development|SetHiddeninGame (Variables|Default|GetMissileMesh) true false)
          (GameplayCue|ExecuteGameplayCue%28Burst%29Local self (select (or (Utilities|Enum|Equal(Enum) (Variables|Default|GetSurfaceType) "SurfaceType8") (Water|IsinWater (Variables|Water|GetWaterInteractionComponent))) (Variables|Default|GetGC%20Explode%20Water) (Variables|Default|GetGC%20Explode)) (Ability|GameplayCue|MakeGameplayCueParameters 0.0)) ; Cam shake + controller vibration
          (Utilities|Time|ClearandInvalidateTimerbyHandle (Variables|Audio|GetTurnSoundTimerHandle))
          (Audio|Components|Audio|FadeOut (Variables|Default|GetRocketAirAudio) 0.1 0.0 "Linear" :!class /Script/Engine.AudioComponent)
          (Utilities|AllowJanus
            (:Allow
              (bind _returnvalue (Development|GetConsoleVariableIntValue "r.PhysicsField.EnableField"))
              (if (== _returnvalue 1)
                (Game|SpawnActorfromClass "/Game/Effects/Niagara/Blueprints/PhsyicsFields/FN_RadialForce.FN_RadialForce_C" (Transformation|GetWorldTransform (Variables|Collision|GetCapsuleComponent)) "Undefined" "MultiplyWithRoot" 0 0.2 0.75 1.5e+03 2.0 0.0 0)))))
        (:False
          (Variables|Default|SetIsHidden? true))))
    (:then_1
      (CallFunction|ApplyImpulse))
    (:then_2
      (Actor|SetLifeSpan (Variables|Default|GetMarker%20Actor) 0.1))))

(event OnEnterWaterEvent(WaterInteractionComponent)
  (params (WaterBody :type Fort%20Water%20Body%20Actor%20Object%20Reference) (WaterInteractionComponent :type Fort%20Water%20Interaction%20Component%20Object%20Reference) (bIsFirstBody :type Boolean))
  ;[0 1 0.08138] Rocket FX and Conveyance
  ;;[1 1 1] Handles Bug where player shoots rocket while submerged
  (if (Variables|Default|GetCanExplodeInWater?)
    (CallFunction|FuseTimerMax)))

(event Custom|CE_SpawnMarkerActor
  (params (TargetPos :type Vector))
  ;[0 1 0.08138] Rocket FX and Conveyance
  ;;[1 1 1] IF using an aim point, spawn marker decal actor
  (Networking|CanExecuteCosmeticEvents
    (:False
      ;;;[1 1 1] If no target, spawn at aim location
      (bind _returnvalue (Game|SpawnActorfromClass "/ResolveGameplay/Items/Guns/HomingRocket/B_HomingRocket_TargetActor.B_HomingRocket_TargetActor_C" (Math|Conversions|ToTransform(Vector) TargetPos) "AlwaysSpawn" "OverrideRootScale" 0 self (Variables|Default|GetMissileMesh) 0 (Game|GetInstigator)))
      (Variables|Default|SetMarker%20Actor _returnvalue))))

(event Custom|UpdateCanExplodeInWater
  ;[0 1 0.08138] Rocket FX and Conveyance
  ;;[1 1 1] Handles Bug where player shoots rocket while submerged
  (Variables|Default|SetCanExplodeInWater? true))

(event Custom|PlayIncomingWarningSound
  ;[0 1 0.08138] Rocket FX and Conveyance
  ;;[1 1 1] Check for the distance from our missile to their target, IF below a certain threshold play incoming sound
  (if (not (Variables|Default|GetIsHidden?))
    (Utilities|IsValid (Variables|Default|GetLock%20on%20Target)
      (:Is%20Valid
        (if (<= (Class|AbilityFunctions|GetDistanceBetweenActors (Variables|Default|GetLock%20on%20Target) self) (Variables|Audio|GetIncomingSoundPlayDistance))
          (Development|PrintString "IN RANGE OF VEHICLE" (Variables|Default|GetDebug) (Variables|Default|GetDebug) "(R=1.000000,G=0.890245,B=0.000000,A=1.000000)" 5.0)
          (Utilities|FlowControl|DoOnce false
            (:Completed
              (Utilities|Time|ClearandInvalidateTimerbyHandle (Variables|Default|GetDistanceCheckTimerHandle))
              (CallFunction|PlayWooshSound)))))
      (:Is%20Not%20Valid
        (if (<= (Class|AbilityFunctions|GetDistanceBetweenActors (Variables|Default|GetMarker%20Actor) self) (Variables|Audio|GetIncomingSoundPlayDistance))
          (Development|PrintString "IN RANGE OF POINT" (Variables|Default|GetDebug) (Variables|Default|GetDebug) "(R=0.000000,G=0.068491,B=1.000000,A=1.000000)" 5.0)
          (Utilities|FlowControl|DoOnce false
            (:Completed
              (Utilities|Time|ClearandInvalidateTimerbyHandle (Variables|Default|GetDistanceCheckTimerHandle))
              (CallFunction|PlayWooshSound))))))))

(event Custom|PlayWooshSound
  ;[0 1 0.08138] Rocket FX and Conveyance
  ;;[1 1 1] Check for the distance from our missile to their target, IF below a certain threshold play incoming sound
  (Audio|SpawnSoundAttached (Variables|Default|GetMissileIncomingWoosh) (Variables|Default|GetRocketAirAudio)))

(event Custom|BankingSoundCheck
  (bind _capsulecomponent (Variables|Collision|GetCapsuleComponent))
  (bind _returnvalue (Transformation|GetForwardVector _capsulecomponent))
  ;[0 1 0.08138] Rocket FX and Conveyance
  ;;[1 1 1] Audio for rocket turning
  (Networking|CanExecuteCosmeticEvents
    (:True
      (if (> (Class|AbilityFunctions|GetAngleBetweenTwoVectors _returnvalue (Variables|Audio|GetLastDirection)) (Variables|Audio|GetBankingSoundDegreeThreshold))
        (Variables|Audio|SetLastDirection _returnvalue)
        (Utilities|FlowControl|DoOnce false
          (:Completed
            (bind _rocketairaudio (Variables|Default|GetRocketAirAudio)
              (:then
                (bind _returnvalue_2 (Audio|SpawnSoundAttached "/ResolveGameplay/Items/Guns/HomingRocket/Sounds/Launcher_HomingRocket_Prj_Turn_Cue.Launcher_HomingRocket_Prj_Turn_Cue" _rocketairaudio))
                (Audio|Components|Audio|FadeOut (Variables|Default|GetActiveTurnSound) 1.0 0.0 "Linear" :!class /Script/Engine.AudioComponent)
                (Variables|Default|SetActiveTurnSound _returnvalue_2)
                (Utilities|Time|SetTimerbyEvent (EventDispatchers|CreateEvent :!event Custom|ResetBankingSound) (Variables|Audio|GetBankingSoundMinRetriggerTime)))
              (:else))))
        (else
          (Variables|Audio|SetLastDirection _returnvalue))))))

(event Custom|ResetBankingSound
  ;[0 1 0.08138] Rocket FX and Conveyance
  ;;[1 1 1] Audio for rocket turning
  (Utilities|FlowControl|DoOnce false
    (:Completed
      (bind _rocketairaudio (Variables|Default|GetRocketAirAudio)
        (:then
          (bind _returnvalue (Audio|SpawnSoundAttached "/ResolveGameplay/Items/Guns/HomingRocket/Sounds/Launcher_HomingRocket_Prj_Turn_Cue.Launcher_HomingRocket_Prj_Turn_Cue" _rocketairaudio))
          (Audio|Components|Audio|FadeOut (Variables|Default|GetActiveTurnSound) 1.0 0.0 "Linear" :!class /Script/Engine.AudioComponent)
          (Variables|Default|SetActiveTurnSound _returnvalue)
          (Utilities|Time|SetTimerbyEvent (EventDispatchers|CreateEvent :!event Custom|ResetBankingSound) (Variables|Audio|GetBankingSoundMinRetriggerTime)))
        (:else)))))

(event Custom|AddWarningWidgetToTargets
  (params (Target :type Object%20Reference))
  ;[0 1 0.08138] Rocket FX and Conveyance
  ;;[1 1 1] Missile Incoming Widget logic
  (bind _asfort_athena_vehicle (Utilities|Casting|CastToFortAthenaVehicle Target))
  (Vehicle|BindEventtoOnPawnFinishEnteringVehicle _asfort_athena_vehicle (delegate Custom|PlayerEnteredVehicle))
  (Vehicle|BindEventtoOnPawnExitedVehicleSeat _asfort_athena_vehicle (delegate Custom|PlayerExitedVehicle))
  ;;;[1 1 1] Get all current passengers and create an array, loop through and apply the widget GE to each
  (bind _returnvalue (Vehicle|GetAllPassengers(Message) Target))
  (bind _output_get (Variables|Default|SetCurrentVehiclePassengers _returnvalue))
  (for _array_element _output_get
    (CallFunction|ApplyLockonWidgetGE _array_element)))

(event Custom|UnbindPlayerEnteredVehicle
  ;[0 1 0.08138] Rocket FX and Conveyance
  ;;[1 1 1] Missile Incoming Widget logic
  ;;;[1 1 1] Added to fix a bug where the missile incoming widget persisted after explode
  (bind _asfort_athena_vehicle (Utilities|Casting|CastToFortAthenaVehicle (Variables|Default|GetLock%20on%20Target)))
  (Vehicle|UnbindEventfromOnPawnFinishEnteringVehicle _asfort_athena_vehicle (EventDispatchers|CreateEvent :!event Custom|PlayerEnteredVehicle)))

(event Custom|PlayerEnteredVehicle
  (params (Vehicle :type Fort%20Vehicle%20Interface%20Interface%20%28by%20ref%29) (PlayerPawn :type Fort%20Pawn%20Object%20Reference) (SeatIndex :type Integer))
  ;[0 1 0.08138] Rocket FX and Conveyance
  ;;[1 1 1] Missile Incoming Widget logic
  ;;;[1 1 1] Add player entering to current passenger array and apply Widget GE
  (CallFunction|ApplyLockonWidgetGE PlayerPawn)
  (Utilities|Array|AddUnique (Variables|Default|GetCurrentVehiclePassengers) PlayerPawn))

(event Custom|PlayerExitedVehicle
  (params (Vehicle :type Fort%20Vehicle%20Interface%20Interface%20%28by%20ref%29) (PlayerPawn :type Fort%20Pawn%20Object%20Reference) (SeatIndex :type Integer))
  ;[0 1 0.08138] Rocket FX and Conveyance
  ;;[1 1 1] Missile Incoming Widget logic
  ;;;[1 1 1] Remove player on exit from current passenger array and remove widget GE
  (Utilities|IsinVehicle PlayerPawn
    (:false
      (Utilities|Array|RemoveItem (Variables|Default|GetCurrentVehiclePassengers) PlayerPawn))))

(fn ApplyLockOnWidgetGE
  (params (Actor :type Actor%20Object%20Reference))
  (if (Networking|IsServer)
    (GameplayEffects|ApplyGameplayEffectToSelf :self (Ability|GetFortAbilitySystemComponent Actor) :GameplayEffectClass "/ResolveGameplay/Items/Guns/HomingRocket/GameplayEffects/GE_HomingRocket_LockWarningHandler.GE_HomingRocket_LockWarningHandler_C" :Level 0.0)))

(fn CalcLaunchVelocity
  (params (Actor :type Actor%20Object%20Reference))
  (return "0, 0, 0"))

(fn GetTargetOffset
  (params (TargetActor :type Actor%20Object%20Reference) (TargetLoc :type Vector))
  (bind _z (.z (Transformation|GetActorLocation TargetActor)))
  (bind (_origin _boxextent) (Collision|GetActorBounds TargetActor true false))
  (bind _returnvalue_x (Math|Vector|BreakVector (Math|Vector|VectorBoundedtoBox (Math|Vector|UnrotateVector (- TargetLoc (Transformation|GetActorLocation TargetActor)) (Transformation|GetActorRotation TargetActor)) (* _boxextent "-1.000000,-1.000000,-1.000000") _boxextent)))
  (bind _returnvalue (Transformation|GetActorLocation TargetActor))
  (bind _z_2 (.z _origin))
  (bind _returnvalue_2 (- _z_2 _z))
  ;[1 1 1] Currently used to check if the actor is a Train
  (Utilities|IsValid (Actor|GetComponentbyClass TargetActor "/Script/FortniteGame.FortMovingPlatformComponent")
    (:Is%20Valid
      (return _returnvalue_x 0.0 _returnvalue_2))
    (:Is%20Not%20Valid
      (return 0.0 0.0 _returnvalue_2))))

(fn GetOwnerVelocity
  (bind _returnvalue (|CallParentFunction))
  (return (Math|Vector|MakeVector (.x _returnvalue) (.y _returnvalue))))

(fn OnRep_IsHidden?
  (Components|Activation|SetActive (Variables|Default|GetNS_HomingRocket_Prj_Trail) false false))

(fn OnRep_MuzzleLoc
  (Transformation|SetWorldLocation (Variables|Collision|GetCapsuleComponent) (Variables|Default|GetMuzzleLoc) false true))

(fn RandomizeNewTargetLocation
  (params (TankLocation :type Vector))
  (return (Utilities|Operators|Add TankLocation (Math|Vector|MakeVector (Math|Float|SelectFloat (Math|Random|RandomFloatinRange 5e+03 8e+03) (Math|Random|RandomFloatinRange -5e+03 -8e+03) (Math|Random|RandomBool)) (Math|Float|SelectFloat (Math|Random|RandomFloatinRange 5e+03 8e+03) (Math|Random|RandomFloatinRange -5e+03 -8e+03) (Math|Random|RandomBool)) 5e+02))))

(fn RemoveOneLockOnStack
  (for _array_element (Variables|Default|GetCurrentVehiclePassengers)
    (Ability|SendGameplayEventtoActor _array_element "(TagName=\"Event.Resolve.Item.Weapon.HomingRocket.RemoveOneLockOnStack\")")))

(fn ConstructionScript
  (|CallParentFunction)
  (Effects|Components|ParticleSystem|SetTemplate (Variables|Default|GetRibbon_Trail_PSC) (Variables|Default|GetRibbon_Trail_PS))
  (Effects|Components|ParticleSystem|SetTemplate (Variables|Default|GetExplosion_Flesh_Damage_PSC) (Variables|Default|GetExplosion_Flesh_Damage_PS)))
