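# Niagara marker effect for the homing rocket, with a vertical ribbon beam and a point light that react to the missile's distance to its target.
# A system-level DistanceLerp module, actor component interface and User.RocketObj input appear to supply MissleDistanceToTarget as a shared alpha.
# That alpha drives beam height, width and material parameter, indexes a shared color curve, and blends the light's intensity from random flicker to steady.
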NiagaraSystem /ResolveGameplay/Items/Guns/HomingRocket/FX/NS_HomingRocket_Marker
dependencies {
 usedDataInterfaces [
  &NiagaraDataInterfaceCurve
  /Script/Niagara.NiagaraDataInterfaceActorComponent
  &NiagaraDataInterfaceColorCurve
 ]
 usedDynamicInputs [&Multiply_VectorByFloat]
 usedModules [
  &self:SystemSpawnScript.SystemScriptSource.SystemScriptGraph.NiagaraNodeAssignment_0.SetVariables_#1
  &SystemState
  &self:DistanceLerp
  &self:SystemSpawnScript.SystemScriptSource.SystemScriptGraph.NiagaraNodeAssignment_1.SetVariables_#2
 ]
}
emitters {
 DynamicBeam {
  emitterUpdateScript [
   &EmitterState(Life Cycle Mode=::NewEnumerator1, Loop Behavior=::NewEnumerator1, Loop Duration Mode=::NewEnumerator1, UseLoopDelay=false)
   /Niagara/Modules/Beams/BeamEmitterSetup(Beam End:Vector3f=&MakeVector(
     X=0
     Y=0
     Z=&ClampFloat(Float=&Lerp_Float(A=3000, B=1000, Alpha=$System.MissleDistanceToTarget), Max=3000)
    ))
   &SpawnBurst_Instantaneous(Spawn Count=2, Use Spawn Probability=false, UseLoopCountLimit=false)
  ]
  particleSpawnScript [
   &InitializeParticle(Position Mode=::NewEnumerator0, Write Material Random=true)
   /Niagara/Modules/Beams/SpawnBeam()
   /Niagara/Modules/Beams/BeamWidth(Beam Width=200)
  ]
  particleUpdateScript [
   &ParticleState(Kill Particles When Lifetime Has Elapsed=false, Let Infinitely Lived Particles Die When Emitter Deactivates=true)
   /Niagara/Modules/Beams/UpdateBeam()
   &SolveForcesAndVelocity(Clamp Velocity=false, Limit Acceleration=false)
   /Niagara/Modules/Beams/ScaleBeamWidth(Beam Width Scale:NiagaraFloat=&Lerp_Float(A=1, B=5, Alpha=&Lerp_Float(A=0, B=1, Alpha=$System.MissleDistanceToTarget)))
   &ScaleColor(
    Curve Index=$System.MissleDistanceToTarget
    Linear Color Curve:NiagaraDataInterfaceColorCurve=$System.ColorCurve
    Scale Mode=::NewEnumerator2
   )
   &DynamicMaterialParameters(
    Index 0 Param 1=&Lerp_Float(A=0, B=1, Alpha=$System.MissleDistanceToTarget)
    Module.Param0WriteEnabled=true
    Module.Param1WriteEnabled=true
    Module.Param2WriteEnabled=true
    Module.Param3WriteEnabled=true
   )
  ]
  rendererClasses [&NiagaraRibbonRendererProperties]
  renderers [{rendererClass=&NiagaraRibbonRendererProperties}]
 }
 Light001 {
  emitterUpdateScript [
   &EmitterState(Life Cycle Mode=::NewEnumerator1, Loop Behavior=::NewEnumerator1, Loop Duration Mode=::NewEnumerator1, UseLoopDelay=false)
   &SpawnBurst_Instantaneous(Use Spawn Probability=false, UseLoopCountLimit=false)
  ]
  particleSpawnScript [
   &InitializeParticle(
    Mesh Scale=<1 60 60>
    Mesh Scale Mode=::NewEnumerator4
    Sprite Size=<1 1>
    Sprite Size Mode=::NewEnumerator4
    UsePositionOffset=false
    Write Material Random=true
   )
   &SystemLocation(MaskSpawnGroup=false, Offset=<0 0 150>)
  ]
  particleUpdateScript [
   /SetVariables(Particles.LightExponent=3) {isSetParameters=true}
   /SetVariables(Particles.LightRadius=1.5) {isSetParameters=true}
   /SetVariables(Particles.LightColorv2:LinearColor=&MultiplyLinearColor(
     Linear Color=&ColorFromCurve(ColorCurve=$System.ColorCurve, CurveIndex=$System.MissleDistanceToTarget)
     Scale Factor=&MakeVector4_FromFloat(Float=&Lerp_Float(
       A=&RandomRangeFloat(Minimum=0, Maximum=20, Evaluation Type=::NewEnumerator1)
       B=20
       Alpha=$System.MissleDistanceToTarget
      ))
    )) {isSetParameters=true}
   /SetVariables(Particles.DecalSize:Vector3f=&Lerp_Vector(A=<1 1 1>, B=<2 2 2>, Alpha=&Lerp_Float(A=1, B=2, Alpha=$System.MissleDistanceToTarget))) {isSetParameters=true}
   &ParticleState(Kill Particles When Lifetime Has Elapsed=false, Let Infinitely Lived Particles Die When Emitter Deactivates=true)
   &ScaleColor(
    Scale Alpha=&Multiply_Float(
     A=1
     B=&FloatFromCurve(
      FloatCurve=DataInterface() {propertyValues {Curve {keys [{tangentMode=RCTM_User, time=0, value=0}, {tangentMode=RCTM_User, time=0.4, value=1}]}}}
      CurveIndex=$Emitter.Age
     )
    )
    ScaleA=true
    ScaleRGB=true
   )
  ]
  rendererClasses [&NiagaraLightRendererProperties]
  renderers [{rendererClass=&NiagaraLightRendererProperties, rendererIndex=0}]
 }
}
summary {
 systemName=NS_HomingRocket_Marker
 emitters [
  {emitterName=DynamicBeam, rendererClasses [&NiagaraRibbonRendererProperties]}
  {emitterName=Light001, rendererClasses [&NiagaraLightRendererProperties]}
 ]
 userVariables [User.RocketObj {
   type=Object
   description=""
   defaultValue {struct=NiagaraExt_VariableValue_Object, value {object=null, objectClass=&Object}}
  }]
}
