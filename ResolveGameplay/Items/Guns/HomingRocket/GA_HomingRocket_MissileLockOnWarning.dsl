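; Gameplay ability that drives the missile lock-on warning for the pawn targeted by a Homing Rocket.
; Cosmetic-capable instances broadcast warning-state messages through the playspace event router to show the UI widget; ability end broadcasts its deactivation.
; A gameplay event wait removes one MissileLocked effect stack when received.
; Non-cosmetic instances end the ability on full vehicle exit, and ending clears the lock-on effect.
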(event Custom|Remove1Stack
  (Ability|RemoveGameplayEffectFromOwnerWithGrantedTags "(GameplayTags=((TagName=\"Granted.Resolve.Item.Weapon.HomingRocket.MissileLocked\")))" 1))

(event Ability|EventActivateAbility
  (Ability|Tasks|WaitGameplayEvent "(TagName=\"Event.Resolve.Item.Weapon.HomingRocket.RemoveOneLockOnStack\")"
    (:then
      (Networking|CanExecuteCosmeticEvents
        (:True
          (CallFunction|GetAvatarasValidFortPlayerPawn
            (:ValidFortPlayerPawn
              ;[1 1 1] Used by UI to activate the widget
              (bind _returnvalue (GameplayEventRouter|GetEventRouter "/Script/FortniteGame.FortPlayspace" (Ability|GetActivatingPawn)))
              (GameplayEventRouter|BroadcastEvent _returnvalue (Utilities|Struct|MakeWarningStateMessage true) 0))))
        (:False
          ;[1 1 1] Kill all lock on stacks on leaving vehicle
          (bind _outfortplayerpawn (CallFunction|GetAvatarasValidFortPlayerPawn)
            (:ValidFortPlayerPawn
              (Utilities|IsValid (Fortnite|GetFortPlayerControllerfromActor _outfortplayerpawn)
                (:Is%20Valid
                  (Vehicle|BindEventtoOnFullyExitVehicle (Fortnite|GetFortPlayerControllerfromActor _outfortplayerpawn) (delegate Custom|ExitVehicle)))))))))
    (:EventReceived
      (CallFunction|Remove1Stack))))

(event Ability|EventOnEndAbility
  (params (bWasCancelled :type Boolean))
  ;[1 1 1] Used by UI to deactivate the widget
  (bind _returnvalue (GameplayEventRouter|GetEventRouter "/Script/FortniteGame.FortPlayspace" (Ability|GetActivatingPawn)))
  (GameplayEventRouter|BroadcastEvent _returnvalue (Utilities|Struct|MakeWarningStateMessage false) 0)
  (Ability|RemoveGameplayEffectFromOwnerWithGrantedTags "(GameplayTags=((TagName=\"Granted.Resolve.Item.Weapon.HomingRocket.MissileLocked\")))"))

(event Custom|ExitVehicle
  ;[1 1 1] Kill all lock on stacks on leaving vehicle
  (Ability|EndAbility))

(event Ability|EventFailedToActivatePassiveAbility
  (params (ActorInfo :type Gameplay%20Ability%20Actor%20Info%20Structure))
  (|CallParentFunction))
