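; Gameplay ability for the homing rocket launcher's lock-on, handling target detection, targeting state and player feedback.
; Indexed camera-relative line traces, offset and angled toward view forward, accept actors tagged as vehicles, quest targets or train cars.
; State changes play weapon targeting sounds and update scope material-collection parameters and the potential target's vehicle stencil.
; A helper broadcasts targeting state, active flag and ammo through the playspace event router.
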(fn BroadcastStateMessage
  (params (IsActive :type Boolean))
  (bind _playerref (Variables|Default|GetPlayerRef))
  (Utilities|IsValid _playerref
    (:Is%20Valid
      (bind _returnvalue (GameplayEventRouter|GetEventRouter "/Script/FortniteGame.FortPlayspace" (Ability|GetActivatingPawn)))
      (GameplayEventRouter|BroadcastEvent _returnvalue (Utilities|Struct|MakeHomingStateMessage (Variables|Default|GetCurrentTargetingState) IsActive (Math|Integer|SelectInt (FortWeapon|GetRemainingAmmo (Weapon|GetCurrentWeapon _playerref)) 0 (Utilities|IsValid (Weapon|GetCurrentWeapon _playerref) :!class /Script/Engine.KismetSystemLibrary)) _playerref) 0))))

(fn ChangeState
  (params (NewState :type ENUM_HomingRocket_LockOnStates%20Enum :default "NewEnumerator0"))
  (if (Utilities|Enum|NotEqual(Enum) NewState (Variables|Default|GetCurrentTargetingState))
    (bind _output_get (Variables|Default|SetCurrentTargetingState (Variables|GetNewState)))
    (Class|BLauncherHomingRocketAthena|PlayTargetingSounds (Variables|Default|GetWeaponRef) _output_get)
    (CallFunction|UpdatePostProcessing)))

(fn ExecuteTraceForLockableTargets
  (params (TracePosition :type Integer))
  (bind (_returnvalue_location _returnvalue_rotation) (Math|Transform|BreakTransform (Targeting|GetTargetingTransform (Variables|Default|GetPlayerRef) "Camera" 0 :!class /Script/FortniteGame.FortPawn)))
  (bind (_output_x _output_y) (Math|Vector2D|BreakVector2D (Utilities|Array|Get(aref) (Variables|Default|GetTargetingLocationOffsets) (Variables|GetTracePosition))))
  (bind _outhit (Collision|LineTraceByChannel :Start (+ _returnvalue_location (* (Math|Vector|Normalize (+ (* (Math|Vector|GetUpVector _returnvalue_rotation) _output_x) (* _output_y (Math|Vector|GetRightVector _returnvalue_rotation)))) (ScalableFloat|GetValueatLevel (Variables|Anvil|ScalableFloats|GetTargetingOffsetSizeMult)))) :End (+ (+ _returnvalue_location (* (Math|Vector|Normalize (+ (* (Math|Vector|GetUpVector _returnvalue_rotation) _output_x) (* _output_y (Math|Vector|GetRightVector _returnvalue_rotation)))) (ScalableFloat|GetValueatLevel (Variables|Anvil|ScalableFloats|GetTargetingOffsetSizeMult)))) (* (Math|Vector|Lerp(Vector) (Math|Vector|Normalize (+ (* (Math|Vector|GetUpVector _returnvalue_rotation) _output_x) (* _output_y (Math|Vector|GetRightVector _returnvalue_rotation)))) (Math|Vector|GetForwardVector _returnvalue_rotation) (ScalableFloat|GetValueatLevel (Variables|Anvil|ScalableFloats|GetTraceAngle))) (ScalableFloat|GetValueatLevel (Variables|Anvil|ScalableFloats|GetTraceDistance)))) :TraceChannel "TraceTypeQuery6" :bTraceComplex false :DrawDebugType (select (Variables|Anvil|Booleans|GetDebug) "ForOneFrame" "None") :bIgnoreSelf true :TraceColor "(R=1.000000,G=0.000000,B=0.000000,A=1.000000)" :TraceHitColor "(R=0.000000,G=1.000000,B=0.000000,A=1.000000)" :DrawTime 5.0))
  ;[1 1 1] Stretch Goal: Gameplay Tag(s) to allow creative to target other things
  (bind _returnvalue (GameplayTags|HasMatchingGameplayTag(Message) (.HitActor (Collision|BreakHitResult _outhit)) "(TagName=\"Vehicle\")"))
  (if _returnvalue
    (return true true true (.HitActor (Collision|BreakHitResult _outhit)))
    (else
      ;[1 1 1] Currently only used for quests. Can be expanded later to handle nonvehicles that care about team or health.
      (bind _returnvalue_2 (GameplayTags|HasMatchingGameplayTag(Message) (.HitActor (Collision|BreakHitResult _outhit)) (Variables|Default|GetQuestTag)))
      (if _returnvalue_2
        (return true false false (.HitActor (Collision|BreakHitResult _outhit)))
        (else
          ;[1 1 1] Used for Train Cars
          (bind _returnvalue_3 (GameplayTags|HasMatchingGameplayTag(Message) (.HitActor (Collision|BreakHitResult _outhit)) (Variables|Default|GetTrainCarTag)))
          (return _returnvalue_3 false false (.HitActor (Collision|BreakHitResult _outhit))))))))

(fn SetNewPotentialTarget
  (params (NewPotentialTarget1 :type Actor%20Object%20Reference)))

(fn UpdatePostProcessing
  (bind _potentialtarget (Variables|Default|GetPotentialTarget))
  (bind _potentialtarget_2 (Variables|Default|GetPotentialTarget))
  (switch Utilities|FlowControl|Switch|SwitchonENUM_HomingRocket_LockOnStates (Variables|Default|GetCurrentTargetingState)
    (:NewEnumerator0
      _)
    (:NewEnumerator1
      (Rendering|Material|SetScalarParameterValue "/Game/Weapons/FORT_Rifles/SniperScope/MPC_WeaponScope.MPC_WeaponScope" "Is Targeting" 1.0 :!class /Script/Engine.KismetMaterialLibrary)
      (Utilities|IsValid _potentialtarget
        (:Is%20Valid
          (Vehicle|EnableVehicleStencil _potentialtarget true true))))
    (:NewEnumerator2
      (Rendering|Material|SetScalarParameterValue "/Game/Weapons/FORT_Rifles/SniperScope/MPC_WeaponScope.MPC_WeaponScope" "Is Locked on" 1.0 :!class /Script/Engine.KismetMaterialLibrary))
    (:NewEnumerator3
      _))
  ;[1 1 1] Turn off parameter values
  (Rendering|Material|SetScalarParameterValue "/Game/Weapons/FORT_Rifles/SniperScope/MPC_WeaponScope.MPC_WeaponScope" "Is Targeting" :!class /Script/Engine.KismetMaterialLibrary)
  (Rendering|Material|SetScalarParameterValue "/Game/Weapons/FORT_Rifles/SniperScope/MPC_WeaponScope.MPC_WeaponScope" "Is Locked on" :!class /Script/Engine.KismetMaterialLibrary)
  (Utilities|IsValid _potentialtarget_2
    (:Is%20Valid
      (Vehicle|EnableVehicleStencil _potentialtarget_2 true))))
