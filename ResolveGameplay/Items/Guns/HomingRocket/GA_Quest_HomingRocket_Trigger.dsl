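; Gameplay ability for the homing rocket gun that reports a homing rocket lock to Athena quest tracking.
; On activation, the server gets the quest manager from the avatar's player controller and sends a complex custom stat event tagged Athena.Quests.HomingRocketLock with a count of one.
; The ability then ends immediately on that server path; the non-server branch does nothing.
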(event Ability|EventActivateAbility
  (Utilities|IsServer
    (:true
      (Quest|SendComplexCustomStatEvent (Quests|GetQuestManager (Fortnite|GetFortPlayerControllerfromActor (Ability|GetAvatarActorfromActorInfo)) "Athena") 0 "(GameplayTags=((TagName=\"Athena.Quests.HomingRocketLock\")))" "(GameplayTags=)" 1)
      (Ability|EndAbility))))
