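# Sound cue for the homing rocket launcher's lock-on start sound while aiming down sights.
# A single wave plays through a modulator that applies slight random pitch variation and a fixed reduced volume.
# It routes to the ranged weapon scope sound class and overrides concurrency so only one instance plays, stopping the oldest.
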SoundCue /ResolveGameplay/Items/Guns/HomingRocket/Sounds/Launcher_HomingRocket_ADS_Locked_Start_Cue
root=Modulator_0
nodes [
 {id=Modulator_0, children [WavePlayer_1], properties {pitchMax=1.02, pitchMin=0.98, volumeMax=0.75, volumeMin=0.75}}
 {id=WavePlayer_1, properties {soundWaveAssetPtr=&/Launcher_HomingRocket_ADS_Locked_Start_01}}
]
properties {
 bOverrideConcurrency=true
 maxDistance=2097152
 soundClassObject=/Game/Sounds/FortSoundClassesAndModes/Fort_WeaponRanged_Scope_PSM
 volumeMultiplier=1
 concurrencyOverrides {maxCount=1, resolutionRule=StopOldest}
}
