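# Sound cue for the homing rocket launcher's looping lock-on tone, named for targeting while aiming down sights.
# A looping wave player feeds an enveloper that fades volume in over a quarter second.
# A concurrency override allows one instance at a time, stopping the oldest, and the cue uses the ranged weapon scope sound class.
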SoundCue /ResolveGameplay/Items/Guns/HomingRocket/Sounds/Launcher_HomingRocket_ADS_TargetingLockingOn_Loop_Cue
root=Enveloper_1
nodes [
 {
  id=Enveloper_1
  children [WavePlayer_2]
  properties {pitchCurve {editorCurveData {keys [{time=0}]}}, volumeCurve {editorCurveData {keys [<at 0 0>,<at 0.25 1>]}}}
 }
 {
  id=WavePlayer_2
  properties {bLooping=true, soundWaveAssetPtr=&/Launcher_HomingRocket_ADS_Locking_Loop}
 }
]
properties {
 bOverrideConcurrency=true
 maxDistance=2097152
 soundClassObject=/Game/Sounds/FortSoundClassesAndModes/Fort_WeaponRanged_Scope_PSM
 volumeMultiplier=1
 concurrencyOverrides {maxCount=1, resolutionRule=StopOldest}
}
