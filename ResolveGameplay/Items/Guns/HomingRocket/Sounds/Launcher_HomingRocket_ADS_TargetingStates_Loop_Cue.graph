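# Sound cue for the homing rocket launcher's aim-down-sights targeting loop, changing with the lock state.
# A switch on the TargetState integer parameter picks silence, a looping default sound or a looping locked sound.
# An enveloper briefly fades in the locked loop, while the default loop plays directly.
# A concurrency override keeps a single instance by stopping the oldest, under a scope-specific ranged weapon sound class.
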SoundCue /ResolveGameplay/Items/Guns/HomingRocket/Sounds/Launcher_HomingRocket_ADS_TargetingStates_Loop_Cue
root=Switch_0
nodes [
 {
  id=Enveloper_2
  children [WavePlayer_4]
  properties {pitchCurve {editorCurveData {keys [{time=0}]}}, volumeCurve {editorCurveData {keys [<at 0 0>,<at 0.25 1>]}}}
 }
 {id=Switch_0, children [null,null,WavePlayer_0,Enveloper_2], properties {intParameterName=TargetState}}
 {
  id=WavePlayer_0
  properties {bLooping=true, soundWaveAssetPtr=&/Launcher_HomingRocket_ADS_Default_Loop}
 }
 {
  id=WavePlayer_4
  properties {bLooping=true, soundWaveAssetPtr=&/Launcher_HomingRocket_ADS_Locked_Loop}
 }
]
properties {
 bOverrideConcurrency=true
 maxDistance=2097152
 soundClassObject=/Game/Sounds/FortSoundClassesAndModes/Fort_WeaponRanged_Scope_PSM
 volumeMultiplier=1
 concurrencyOverrides {maxCount=1, resolutionRule=StopOldest}
}
