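# First-person fire sound cue for the homing rocket launcher.
# A Random node picks one of three equally weighted fire waves and feeds a Modulator that holds volume constant, leaving its default slight pitch variation.
# The cue uses a first-person weapon projectile attenuation asset, a ranged-fire 1P sound class and the shared first-person gunfire concurrency set.
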SoundCue /ResolveGameplay/Items/Guns/HomingRocket/Sounds/Launcher_HomingRocket_Fire_1P_Cue
root=Modulator_0
nodes [
 {id=Modulator_0, children [Random_1], properties {volumeMax=1, volumeMin=1}}
 {id=Random_1, children [WavePlayer_0,WavePlayer_5,WavePlayer_6], properties {bShouldExcludeFromBranchCulling=true, weights [1,1,1]}}
 {id=WavePlayer_0, properties {soundWaveAssetPtr=&/Launcher_HomingRocket_Fire_1P_01}}
 {id=WavePlayer_5, properties {soundWaveAssetPtr=&/Launcher_HomingRocket_Fire_1P_02}}
 {id=WavePlayer_6, properties {soundWaveAssetPtr=&/Launcher_HomingRocket_Fire_1P_03}}
]
properties {
 attenuationSettings=&Fort_Weapon_Projectile_FirstPerson_Athena_Att
 maxDistance=2097152
 soundClassObject=&Fort_WeaponRanged_Fire_1P_PSM
 volumeMultiplier=1
 attenuationOverrides {falloffDistance=5000, attenuationShapeExtents {x=1000}}
 concurrencyOverrides {bLimitToOwner=true, maxCount=3, resolutionRule=StopOldest}
 concurrencySet [&Weapon_Gunfire_FirstPerson_Conc]
}
