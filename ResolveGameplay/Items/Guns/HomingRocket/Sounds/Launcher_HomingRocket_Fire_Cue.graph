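# Sound Cue for the homing rocket launcher's fire sound, built from close, mid-distance and far-distance layers.
# A Mixer blends the layers, weighting the far layer loudest, and each runs through its own SoundClass override and projectile attenuation preset.
# Each layer picks randomly from three waves through a Modulator, with the close layer reusing 1P recordings under a 3P sound class.
# Local and global gunfire concurrency sets limit playback.
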SoundCue /ResolveGameplay/Items/Guns/HomingRocket/Sounds/Launcher_HomingRocket_Fire_Cue
root=Mixer_3
nodes [
 {
  id=Attenuation_10
  children [Modulator_2]
  properties {attenuationSettings=&Fort_Weapon_Projectile_Medium_Mono_Athena_Att}
 }
 {
  id=Attenuation_11
  children [Modulator_1]
  properties {attenuationSettings=&Fort_Weapon_Projectile_Distant_Mono_Athena_Att}
 }
 {id=Attenuation_9, children [Modulator_3], properties {attenuationSettings=&Fort_Weapon_Projectile_Close_Athena_Att}}
 {id=Mixer_3, children [SoundClass_7,SoundClass_4,SoundClass_3], properties {inputVolume [1,1.5,2.5]}}
 {id=Modulator_1, children [Random_1]}
 {id=Modulator_2, children [Random_2]}
 {id=Modulator_3, children [Random_5]}
 {id=Random_1, children [WavePlayer_5,WavePlayer_6,WavePlayer_7], properties {bShouldExcludeFromBranchCulling=true, weights [1,1,1]}}
 {id=Random_2, children [WavePlayer_8, WavePlayer_9, WavePlayer_10], properties {bShouldExcludeFromBranchCulling=true, weights [1,1,1]}}
 {id=Random_5, children [WavePlayer_11, WavePlayer_18, WavePlayer_19], properties {bShouldExcludeFromBranchCulling=true, weights [1,1,1]}}
 {
  id=SoundClass_3
  children [Attenuation_11]
  properties {soundClassOverride=&Fort_WeaponRanged_Fire_Distant}
 }
 {
  id=SoundClass_4
  children [Attenuation_10]
  properties {soundClassOverride=&Fort_WeaponRanged_Fire_Distant}
 }
 {
  id=SoundClass_7
  children [Attenuation_9]
  properties {soundClassOverride=&Fort_WeaponRanged_Fire_3P_PSM}
 }
 {id=WavePlayer_10, properties {soundWaveAssetPtr=&/Launcher_HomingRocket_Fire_Mid_Distance_03}}
 {id=WavePlayer_11, properties {soundWaveAssetPtr=&/Launcher_HomingRocket_Fire_1P_01}}
 {id=WavePlayer_18, properties {soundWaveAssetPtr=&/Launcher_HomingRocket_Fire_1P_02}}
 {id=WavePlayer_19, properties {soundWaveAssetPtr=&/Launcher_HomingRocket_Fire_1P_03}}
 {id=WavePlayer_5, properties {soundWaveAssetPtr=&/Launcher_HomingRocket_Fire_Far_Distance_01}}
 {id=WavePlayer_6, properties {soundWaveAssetPtr=&/Launcher_HomingRocket_Fire_Far_Distance_02}}
 {id=WavePlayer_7, properties {soundWaveAssetPtr=&/Launcher_HomingRocket_Fire_Far_Distance_03}}
 {id=WavePlayer_8, properties {soundWaveAssetPtr=&/Launcher_HomingRocket_Fire_Mid_Distance_01}}
 {id=WavePlayer_9, properties {soundWaveAssetPtr=&/Launcher_HomingRocket_Fire_Mid_Distance_02}}
]
properties {
 maxDistance=27500
 volumeMultiplier=0.5
 attenuationOverrides {falloffDistance=5000, attenuationShapeExtents {x=1000}}
 concurrencyOverrides {resolutionRule=PreventNew}
 concurrencySet [&Weapon_Gunfire_Local_Conc, &Weapon_Gunfire_Global_Conc]
}
