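# Sound cue for the homing rocket launcher's impact area flare, played with distance attenuation under a third-person ranged weapon fire sound class.
# A quality switch sends its first two levels to a mixer that layers a one-shot start sound with a looping flare sound.
# The last two levels play only the start sound, with no loop.
# A concurrency override caps simultaneous instances at two.
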SoundCue /ResolveGameplay/Items/Guns/HomingRocket/Sounds/Launcher_HomingRocket_Impact_Area_Loop_Cue
root=QualityLevel_0
nodes [
 {id=Mixer_0, children [WavePlayer_2,WavePlayer_1,null], properties {inputVolume [1,1.2,1]}}
 {id=Mixer_1, children [WavePlayer_3,null], properties {inputVolume [1,1]}}
 {id=QualityLevel_0, children [Mixer_0,Mixer_0,Mixer_1,Mixer_1]}
 {
  id=WavePlayer_1
  properties {bLooping=true, soundWaveAssetPtr=&/Launcher_HomingRocket_Impact_Flare_Loop_01}
 }
 {
  id=WavePlayer_2
  properties {soundWaveAssetPtr=&/Launcher_HomingRocket_Impact_Flare_Loop_Start_01}
 }
 {
  id=WavePlayer_3
  properties {soundWaveAssetPtr=&Launcher_HomingRocket_Impact_Flare_Loop_Start_01}
 }
]
properties {
 attenuationSettings=&/Launcher_HomingRocket_Impact_Area_Att
 bOverrideConcurrency=true
 maxDistance=3600.0002
 soundClassObject=&Fort_WeaponRanged_Fire_3P_PSM
 volumeMultiplier=1.1
 concurrencyOverrides {maxCount=2}
}
