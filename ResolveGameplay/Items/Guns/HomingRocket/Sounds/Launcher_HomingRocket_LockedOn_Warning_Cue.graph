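# Sound cue that plays the warning when a homing rocket launcher has locked on.
# A mixer layers a one-shot lock-on start sound, at reduced volume, with a looping projectile alarm.
# It uses a critical in-game UI sound class with gadget-activation attenuation and boosted overall volume.
# A concurrency override allows only one instance at a time, stopping the oldest.
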SoundCue /ResolveGameplay/Items/Guns/HomingRocket/Sounds/Launcher_HomingRocket_LockedOn_Warning_Cue
root=Mixer_1
nodes [
 {id=Mixer_1, children [WavePlayer_3,WavePlayer_0], properties {inputVolume [0.4,1.1]}}
 {
  id=WavePlayer_0
  properties {bLooping=true, soundWaveAssetPtr=&/Launcher_HomingRocket_Prj_POI_Alarm}
 }
 {id=WavePlayer_3, properties {soundWaveAssetPtr=/FNGameplayCues/Sounds/Fort_Weapons/Launchers/MissileBattery/MistyBop_LockOnCopter_Start_02}}
]
properties {
 attenuationSettings=&Fort_GadgetActivate_Att
 bOverrideConcurrency=true
 maxDistance=4000
 soundClassObject=&Fort_UI_InGame_Critical
 volumeMultiplier=2
 concurrencyOverrides {maxCount=1, resolutionRule=StopOldest}
}
