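# Sound cue named for the low-ammo sound of the homing rocket launcher.
# It has no root and no nodes, so the graph references no audio.
# It sets only playback properties: close-range weapon projectile attenuation, a maximum distance, and a ranged-weapon low-ammo sound class.
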SoundCue /ResolveGameplay/Items/Guns/HomingRocket/Sounds/Launcher_HomingRocket_LowAmmo_Cue
root=null
nodes []
properties {
 attenuationSettings=&Fort_Weapon_Projectile_Close_Athena_Att
 maxDistance=3055.5525
 soundClassObject=/Game/Sounds/FortSoundClassesAndModes/Fort_WeaponRanged_LowAmmo_Degraded_PSM
}
