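# Sound cue for the out-of-ammo sound of the homing rocket launcher in ResolveGameplay.
# A single wave player feeds the Launcher_HomingRocket_OutOfAmmo sound wave straight to the output, with no randomization or other nodes.
# The cue plays quieter and slightly higher-pitched than the source, has a very large max distance, and uses a ranged-weapon low-ammo sound class.
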SoundCue /ResolveGameplay/Items/Guns/HomingRocket/Sounds/Launcher_HomingRocket_OutOfAmmo_Cue
root=WavePlayer_0
nodes [{id=WavePlayer_0, properties {soundWaveAssetPtr=&/Launcher_HomingRocket_OutOfAmmo}}]
properties {
 maxDistance=2097152
 pitchMultiplier=1.2
 soundClassObject=/Game/Sounds/FortSoundClassesAndModes/Fort_WeaponRanged_LowAmmo_Degraded_PSM
 volumeMultiplier=0.4
}
