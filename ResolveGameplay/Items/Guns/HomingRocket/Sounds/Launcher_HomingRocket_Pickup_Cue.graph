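# Sound cue that plays when the homing rocket launcher is picked up.
# A single pickup wave feeds a Modulator that randomizes pitch on each play.
# It uses the shared weapon pickup attenuation and loot collect sound class at reduced volume, and references a pickup concurrency asset alongside a single-instance, prevent-new override.
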SoundCue /ResolveGameplay/Items/Guns/HomingRocket/Sounds/Launcher_HomingRocket_Pickup_Cue
root=Modulator_0
nodes [
 {id=Modulator_0, children [WavePlayer_0], properties {pitchMax=1.1}}
 {id=WavePlayer_0, properties {soundWaveAssetPtr=&/Launcher_HomingRocket_Pickup_01}}
]
properties {
 attenuationSettings=&Fort_Weapon_Pickup_Att
 maxDistance=3500
 soundClassObject=&Fort_Loot_Collect
 volumeMultiplier=0.478
 concurrencyOverrides {maxCount=1, resolutionRule=PreventNew}
 concurrencySet [&PickupMaximum_Oldest]
}
