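# Sound cue for the homing rocket launcher's projectile explosion, layering close, echo tail, medium and distant sounds by distance.
# Each layer randomly picks one of three waves, varies its pitch, and passes through its own attenuation and sound class before mixing.
# The echo tail plays only on the first quality-level branch, and the distant layer gets a short random delay.
# Concurrency is overridden to cap simultaneous instances.
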SoundCue /ResolveGameplay/Items/Guns/HomingRocket/Sounds/Launcher_HomingRocket_Prj_Expl_Cue
root=Mixer_5
nodes [
 {
  id=Attenuation_0
  children [Modulator_0]
  properties {
   attenuationSettings=/Game/Sounds/Attenuation/Fort_Explosion_Close_Tails_Att
   attenuationOverrides {falloffDistance=900, attenuationShapeExtents {x=100}}
  }
 }
 {
  id=Attenuation_1
  children [Modulator_3]
  properties {
   attenuationSettings=/Game/Sounds/Attenuation/Fort_Explosion_Close_Athena_Att
   attenuationOverrides {falloffDistance=900, attenuationShapeExtents {x=100}}
  }
 }
 {id=Attenuation_2, children [Modulator_2], properties {attenuationSettings=/Game/Sounds/Attenuation/Fort_Explosion_Medium_Athena_Att}}
 {
  id=Attenuation_8
  children [Delay_0]
  properties {
   attenuationSettings=/Game/Sounds/Attenuation/Fort_Explosion_Distant_Athena_Att
   attenuationOverrides {falloffDistance=3000, attenuationShapeExtents {x=1000}}
  }
 }
 {id=Delay_0, children [Modulator_7], properties {delayMax=0.1}}
 {id=Mixer_0, children [SoundClass_1, QualityLevel_0], properties {inputVolume [1,1]}}
 {id=Mixer_5, children [Mixer_0,SoundClass_4,SoundClass_5], properties {inputVolume [1.3,1.3,1.3]}}
 {id=Modulator_0, children [Random_3], properties {pitchMax=1.1, pitchMin=0.9, volumeMax=1, volumeMin=1}}
 {id=Modulator_2, children [Random_1], properties {pitchMax=1.1, pitchMin=0.9}}
 {id=Modulator_3, children [Random_0], properties {pitchMax=1.1, pitchMin=0.9, volumeMax=1, volumeMin=1}}
 {id=Modulator_7, children [Random_2], properties {pitchMax=1.15, pitchMin=0.85}}
 {id=QualityLevel_0, children [SoundClass_0,null,null,null]}
 {id=Random_0, children [WavePlayer_0,WavePlayer_1,WavePlayer_2], properties {bShouldExcludeFromBranchCulling=true, weights [1,1,1]}}
 {id=Random_1, children [WavePlayer_3,WavePlayer_4,WavePlayer_5], properties {bShouldExcludeFromBranchCulling=true, weights [1,1,1]}}
 {id=Random_2, children [WavePlayer_6,WavePlayer_7,WavePlayer_8], properties {bShouldExcludeFromBranchCulling=true, weights [1,1,1]}}
 {id=Random_3, children [WavePlayer_9, WavePlayer_10, WavePlayer_11], properties {bShouldExcludeFromBranchCulling=true, weights [1,1,1]}}
 {
  id=SoundClass_0
  children [Attenuation_0]
  properties {soundClassOverride=&Fort_WeaponRanged_Fire_3P_PSM}
 }
 {
  id=SoundClass_1
  children [Attenuation_1]
  properties {soundClassOverride=&Fort_WeaponRanged_Fire_3P_PSM}
 }
 {
  id=SoundClass_4
  children [Attenuation_2]
  properties {soundClassOverride=&Fort_WeaponRanged_Fire_Distant}
 }
 {
  id=SoundClass_5
  children [Attenuation_8]
  properties {soundClassOverride=&Fort_WeaponRanged_Fire_Distant}
 }
 {id=WavePlayer_0, properties {soundWaveAssetPtr=&/Launcher_HomingRocket_Prj_Explo_Close_01}}
 {id=WavePlayer_1, properties {soundWaveAssetPtr=&/Launcher_HomingRocket_Prj_Explo_Close_02}}
 {
  id=WavePlayer_10
  properties {soundWaveAssetPtr=&/Launcher_HomingRocket_Prj_Explo_Close_2D_Echo_02}
 }
 {
  id=WavePlayer_11
  properties {soundWaveAssetPtr=&/Launcher_HomingRocket_Prj_Explo_Close_2D_Echo_03}
 }
 {id=WavePlayer_2, properties {soundWaveAssetPtr=&/Launcher_HomingRocket_Prj_Explo_Close_03}}
 {id=WavePlayer_3, properties {soundWaveAssetPtr=&/Launcher_HomingRocket_Prj_Explo_Medium_01}}
 {id=WavePlayer_4, properties {soundWaveAssetPtr=&/Launcher_HomingRocket_Prj_Explo_Medium_02}}
 {id=WavePlayer_5, properties {soundWaveAssetPtr=&/Launcher_HomingRocket_Prj_Explo_Medium_03}}
 {id=WavePlayer_6, properties {soundWaveAssetPtr=&/Launcher_HomingRocket_Prj_Explo_Distant_01}}
 {id=WavePlayer_7, properties {soundWaveAssetPtr=&/Launcher_HomingRocket_Prj_Explo_Distant_02}}
 {id=WavePlayer_8, properties {soundWaveAssetPtr=&/Launcher_HomingRocket_Prj_Explo_Distant_03}}
 {
  id=WavePlayer_9
  properties {soundWaveAssetPtr=&/Launcher_HomingRocket_Prj_Explo_Close_2D_Echo_01}
 }
]
properties {
 bOverrideConcurrency=true
 maxDistance=27500
 soundClassObject=&Fort_WeaponRanged_Fire_3P_PSM
 volumeMultiplier=1
 concurrencyOverrides {maxCount=3}
 concurrencySet [/Game/Sounds/Concurrency_Rules/Weapon_Explosion_Conc]
}
