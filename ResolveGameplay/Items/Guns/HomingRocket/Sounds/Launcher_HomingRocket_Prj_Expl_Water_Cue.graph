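# Sound cue for the homing rocket launcher projectile's water explosion, layered by distance with a concurrency limit.
# A top mixer blends close, medium and distant branches, each with its own attenuation, sound class override and pitch-modulated random pick of three waves.
# The distant branch adds a slight random delay.
# The close branch also mixes in a quality-gated tails layer, but its modulator has no wave connected.
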SoundCue /ResolveGameplay/Items/Guns/HomingRocket/Sounds/Launcher_HomingRocket_Prj_Expl_Water_Cue
root=Mixer_5
nodes [
 {
  id=Attenuation_2
  children [Modulator_3]
  properties {
   attenuationSettings=/Game/Sounds/Attenuation/Fort_Explosion_Close_Athena_Att
   attenuationOverrides {falloffDistance=900, attenuationShapeExtents {x=100}}
  }
 }
 {id=Attenuation_3, children [Modulator_2], properties {attenuationSettings=/Game/Sounds/Attenuation/Fort_Explosion_Medium_Athena_Att}}
 {
  id=Attenuation_6
  children [Modulator_4]
  properties {
   attenuationSettings=/Game/Sounds/Attenuation/Fort_Explosion_Close_Tails_Att
   attenuationOverrides {falloffDistance=900, attenuationShapeExtents {x=100}}
  }
 }
 {
  id=Attenuation_8
  children [Delay_0]
  properties {
   attenuationSettings=/Game/Sounds/Attenuation/Fort_Explosion_Distant_Athena_Att
   attenuationOverrides {falloffDistance=3000, attenuationShapeExtents {x=1000}}
  }
 }
 {id=Delay_0, children [Modulator_7], properties {delayMax=0.1}}
 {id=Mixer_1, children [SoundClass_1, QualityLevel_0], properties {inputVolume [1,1]}}
 {id=Mixer_5, children [Mixer_1,SoundClass_4,SoundClass_5], properties {inputVolume [1.3,1.45,1]}}
 {id=Modulator_2, children [Random_11], properties {pitchMax=1.1, pitchMin=0.9}}
 {id=Modulator_3, children [Random_9], properties {pitchMax=1.1, pitchMin=0.9, volumeMax=1, volumeMin=1}}
 {id=Modulator_4, children [null], properties {pitchMax=1.1, pitchMin=0.9, volumeMax=1, volumeMin=1}}
 {id=Modulator_7, children [Random_10], properties {pitchMax=1.15, pitchMin=0.85}}
 {id=QualityLevel_0, children [SoundClass_6,null,null,null]}
 {id=Random_10, children [WavePlayer_37, WavePlayer_38, WavePlayer_39], properties {bShouldExcludeFromBranchCulling=true, weights [1,1,1]}}
 {id=Random_11, children [WavePlayer_45, WavePlayer_46, WavePlayer_47], properties {bShouldExcludeFromBranchCulling=true, weights [1,1,1]}}
 {id=Random_9, children [WavePlayer_34, WavePlayer_35, WavePlayer_36], properties {weights [1,1,1]}}
 {
  id=SoundClass_1
  children [Attenuation_2]
  properties {soundClassOverride=&Fort_WeaponRanged_Fire_3P_PSM}
 }
 {
  id=SoundClass_4
  children [Attenuation_3]
  properties {soundClassOverride=&Fort_WeaponRanged_Fire_Distant}
 }
 {
  id=SoundClass_5
  children [Attenuation_8]
  properties {soundClassOverride=&Fort_WeaponRanged_Fire_Distant}
 }
 {
  id=SoundClass_6
  children [Attenuation_6]
  properties {soundClassOverride=&Fort_WeaponRanged_Fire_3P_PSM}
 }
 {
  id=WavePlayer_34
  properties {soundWaveAssetPtr=&/Launcher_HomingRocket_Prj_Explo_Close_Water_01}
 }
 {
  id=WavePlayer_35
  properties {soundWaveAssetPtr=&/Launcher_HomingRocket_Prj_Explo_Close_Water_02}
 }
 {
  id=WavePlayer_36
  properties {soundWaveAssetPtr=&/Launcher_HomingRocket_Prj_Explo_Close_Water_03}
 }
 {
  id=WavePlayer_37
  properties {soundWaveAssetPtr=&/Launcher_HomingRocket_Prj_Explo_Distant_Water_03}
 }
 {
  id=WavePlayer_38
  properties {soundWaveAssetPtr=&/Launcher_HomingRocket_Prj_Explo_Distant_Water_01}
 }
 {
  id=WavePlayer_39
  properties {soundWaveAssetPtr=&/Launcher_HomingRocket_Prj_Explo_Distant_Water_02}
 }
 {
  id=WavePlayer_45
  properties {soundWaveAssetPtr=&/Launcher_HomingRocket_Prj_Explo_Medium_Water_01}
 }
 {
  id=WavePlayer_46
  properties {soundWaveAssetPtr=&/Launcher_HomingRocket_Prj_Explo_Medium_Water_02}
 }
 {
  id=WavePlayer_47
  properties {soundWaveAssetPtr=&/Launcher_HomingRocket_Prj_Explo_Medium_Water_03}
 }
]
properties {
 bOverrideConcurrency=true
 maxDistance=27500
 soundClassObject=&Fort_WeaponRanged_Fire_3P_PSM
 volumeMultiplier=0.8
 concurrencyOverrides {maxCount=3}
 concurrencySet [/Game/Sounds/Concurrency_Rules/Weapon_Explosion_Conc]
}
