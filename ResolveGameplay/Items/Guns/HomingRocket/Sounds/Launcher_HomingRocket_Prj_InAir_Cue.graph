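# Sound cue for the homing rocket projectile in flight, combining launch ascent one-shots with looping in-air layers.
# Close, mid and distant loops each have their own attenuation, mixed for distance-based layering.
# Envelopers fade the loops in over the first second, with a pitch ramp on the close loop.
# A start beep shares the close attenuation, while one of three random ascend sounds plays after a short delay through its own attenuation.
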SoundCue /ResolveGameplay/Items/Guns/HomingRocket/Sounds/Launcher_HomingRocket_Prj_InAir_Cue
root=Mixer_0
nodes [
 {
  id=Attenuation_0
  children [Mixer_1]
  properties {attenuationSettings=&/Launcher_HomingRocket_Prj_InAir_Close_Loop_Att}
 }
 {
  id=Attenuation_1
  children [Enveloper_2]
  properties {attenuationSettings=&/Launcher_HomingRocket_Prj_InAir_Distant_Loop_Att}
 }
 {
  id=Attenuation_2
  children [Enveloper_0]
  properties {attenuationSettings=&/Launcher_HomingRocket_Prj_InAir_Mid_Loop_Att}
 }
 {
  id=Attenuation_3
  children [Delay_1]
  properties {attenuationSettings=&/Launcher_HomingRocket_Prj_InAir_Ascend_Att}
 }
 {id=Delay_0, children [Modulator_3]}
 {id=Delay_1, children [Random_0], properties {delayMax=0.3, delayMin=0.3}}
 {
  id=Enveloper_0
  children [Modulator_1]
  properties {pitchCurve {editorCurveData {keys [{time=0}]}}, volumeCurve {editorCurveData {keys [<at 0 0>,<at 0.75 0>,<at 1 1>]}}}
 }
 {
  id=Enveloper_1
  children [Modulator_0]
  properties {
   pitchCurve {editorCurveData {keys [<at 0 0.8 RCIM_Cubic>, {time=1.5}]}}
   volumeCurve {editorCurveData {keys [<at 0 0>, {interpMode=RCIM_Cubic, leaveTangent=2, time=0.35, value=0}, <at 1 1>]}}
  }
 }
 {
  id=Enveloper_2
  children [Modulator_2]
  properties {pitchCurve {editorCurveData {keys [{time=0}]}}, volumeCurve {editorCurveData {keys [<at 0 0>,<at 0.75 0>,<at 1 1>]}}}
 }
 {id=Mixer_0, children [Mixer_2, Attenuation_2, Attenuation_1, null], properties {inputVolume [1,0.8,0.5,1.5]}}
 {id=Mixer_1, children [Enveloper_1,Delay_0], properties {inputVolume [2.25,1.2]}}
 {id=Mixer_2, children [Attenuation_0, Attenuation_3], properties {inputVolume [1,1.2]}}
 {id=Modulator_0, children [WavePlayer_3], properties {volumeMax=1, volumeMin=1}}
 {id=Modulator_1, children [WavePlayer_1], properties {volumeMax=1, volumeMin=1}}
 {id=Modulator_2, children [WavePlayer_2], properties {volumeMax=1, volumeMin=1}}
 {id=Modulator_3, children [WavePlayer_4], properties {pitchMax=1, pitchMin=1, volumeMax=1, volumeMin=1}}
 {id=Random_0, children [WavePlayer_5,WavePlayer_6,WavePlayer_7], properties {bShouldExcludeFromBranchCulling=true, weights [1,1,1]}}
 {
  id=WavePlayer_1
  properties {bLooping=true, soundWaveAssetPtr=&/Launcher_HomingRocket_Prj_InAir_Mid_Loop}
 }
 {
  id=WavePlayer_2
  properties {bLooping=true, soundWaveAssetPtr=&/Launcher_HomingRocket_Prj_InAir_Distant_Loop}
 }
 {
  id=WavePlayer_3
  properties {bLooping=true, soundWaveAssetPtr=&/Launcher_HomingRocket_Prj_InAir_Close_Loop}
 }
 {
  id=WavePlayer_4
  properties {soundWaveAssetPtr=&/Launcher_HomingRocket_Prj_Ascend_Start_Beep_01}
 }
 {id=WavePlayer_5, properties {soundWaveAssetPtr=&/Launcher_HomingRocket_Prj_Ascend_Start_01}}
 {id=WavePlayer_6, properties {soundWaveAssetPtr=&/Launcher_HomingRocket_Prj_Ascend_Start_02}}
 {id=WavePlayer_7, properties {soundWaveAssetPtr=&/Launcher_HomingRocket_Prj_Ascend_Start_03}}
]
properties {
 bOverrideConcurrency=true
 maxDistance=27500
 soundClassObject=&Fort_WeaponRanged_Fire_3P_PSM
 volumeMultiplier=1
 attenuationOverrides {falloffDistance=5000, attenuationShapeExtents {x=1000}}
 concurrencyOverrides {bLimitToOwner=true, maxCount=3, volumeScale=0.5}
}
