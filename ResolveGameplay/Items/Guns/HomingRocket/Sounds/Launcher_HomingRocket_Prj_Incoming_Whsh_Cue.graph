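# Sound Cue for the incoming whoosh of a homing rocket launcher projectile.
# A random node picks one of three equally weighted whoosh wave variations, played with raised pitch and slightly reduced volume.
# It uses a dedicated attenuation asset and a third-person ranged weapon fire sound class.
# Overridden concurrency caps it at a few instances per owner with volume scaling.
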SoundCue /ResolveGameplay/Items/Guns/HomingRocket/Sounds/Launcher_HomingRocket_Prj_Incoming_Whsh_Cue
root=Random_1
nodes [
 {id=Random_1, children [WavePlayer_0,WavePlayer_8,WavePlayer_9], properties {bShouldExcludeFromBranchCulling=true, weights [1,1,1]}}
 {id=WavePlayer_0, properties {soundWaveAssetPtr=&/Launcher_HomingRocket_Prj_Incoming_Whsh_01}}
 {id=WavePlayer_8, properties {soundWaveAssetPtr=&/Launcher_HomingRocket_Prj_Incoming_Whsh_02}}
 {id=WavePlayer_9, properties {soundWaveAssetPtr=&/Launcher_HomingRocket_Prj_Incoming_Whsh_03}}
]
properties {
 attenuationSettings=&/Launcher_HomingRocket_Prj_Incoming_Whsh_Att
 bOverrideConcurrency=true
 maxDistance=10800
 pitchMultiplier=1.2
 soundClassObject=&Fort_WeaponRanged_Fire_3P_PSM
 volumeMultiplier=0.8
 attenuationOverrides {falloffDistance=5000, attenuationShapeExtents {x=1000}}
 concurrencyOverrides {bLimitToOwner=true, maxCount=3, volumeScale=0.5}
}
