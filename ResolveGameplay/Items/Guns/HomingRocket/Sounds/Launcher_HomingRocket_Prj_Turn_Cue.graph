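# Sound cue for the turning sound of a homing rocket launcher's projectile, classed as third-person ranged weapon fire.
# A quality-level switch feeds the first two slots into a shared attenuation node, leaving the other two empty.
# That node wraps an equally weighted random pick of three turn waves, excluded from branch culling.
# Overridden concurrency limits voices per owner, stopping the oldest and scaling down volume as instances stack.
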SoundCue /ResolveGameplay/Items/Guns/HomingRocket/Sounds/Launcher_HomingRocket_Prj_Turn_Cue
root=QualityLevel_0
nodes [
 {
  id=Attenuation_4
  children [Random_1]
  properties {attenuationSettings=&/Launcher_HomingRocket_Prj_InAir_Turn_Att}
 }
 {id=QualityLevel_0, children [Attenuation_4, Attenuation_4, null, null]}
 {id=Random_1, children [WavePlayer_0,WavePlayer_8,WavePlayer_9], properties {bShouldExcludeFromBranchCulling=true, weights [1,1,1]}}
 {id=WavePlayer_0, properties {soundWaveAssetPtr=&/Launcher_HomingRocket_Prj_Turn_01}}
 {id=WavePlayer_8, properties {soundWaveAssetPtr=&/Launcher_HomingRocket_Prj_Turn_02}}
 {id=WavePlayer_9, properties {soundWaveAssetPtr=&/Launcher_HomingRocket_Prj_Turn_03}}
]
properties {
 bOverrideConcurrency=true
 maxDistance=16500
 soundClassObject=&Fort_WeaponRanged_Fire_3P_PSM
 volumeMultiplier=0.5
 attenuationOverrides {falloffDistance=5000, attenuationShapeExtents {x=1000}}
 concurrencyOverrides {bLimitToOwner=true, maxCount=2, resolutionRule=StopOldest, voiceStealReleaseTime=0.5, volumeScale=0.5, volumeScaleAttackTime=0.5}
}
