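# Sound cue for the end of the homing rocket launcher's reload, classed as weapon foley with reload attenuation.
# A mixer plays two one-shot waves layered: the A sound immediately and the B sound after a short fixed delay.
# Each layer passes through its own modulator for slight random pitch variation.
# Concurrency is overridden to one instance, stopping the farthest before preventing new ones.
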SoundCue /ResolveGameplay/Items/Guns/HomingRocket/Sounds/Launcher_HomingRocket_ReloadEnd_Cue
root=Mixer_0
nodes [
 {id=Delay_0, children [Modulator_1], properties {delayMax=0.2, delayMin=0.2}}
 {id=Mixer_0, children [Modulator_0,Delay_0], properties {inputVolume [1,1]}}
 {id=Modulator_0, children [WavePlayer_0], properties {pitchMax=1.1, volumeMax=1, volumeMin=1}}
 {id=Modulator_1, children [WavePlayer_1], properties {pitchMax=1.1, volumeMax=1, volumeMin=1}}
 {id=WavePlayer_0, properties {soundWaveAssetPtr=&/Launcher_HomingRocket_ReloadEnd_A_01}}
 {id=WavePlayer_1, properties {soundWaveAssetPtr=&/Launcher_HomingRocket_ReloadEnd_B_01}}
]
properties {
 attenuationSettings=&Fort_Weapon_Reload_Att
 bOverrideConcurrency=true
 maxDistance=2250
 soundClassObject=&Fort_WeaponFoley
 volumeMultiplier=0.351
 attenuationOverrides {falloffDistance=5000, attenuationShapeExtents {x=1000}}
 concurrencyOverrides {maxCount=1, resolutionRule=StopFarthestThenPreventNew}
}
