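# Sound cue for the reload insert sound of the homing rocket launcher, classed as weapon foley.
# A single wave player feeds a modulator that randomizes pitch on each play while holding volume constant.
# It uses the shared weapon reload attenuation at a reduced volume multiplier.
# A concurrency override limits it to one instance, stopping the farthest and then preventing new plays.
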SoundCue /ResolveGameplay/Items/Guns/HomingRocket/Sounds/Launcher_HomingRocket_ReloadInsert_Cue
root=Modulator_0
nodes [
 {id=Modulator_0, children [WavePlayer_0], properties {pitchMax=1.1, volumeMax=1, volumeMin=1}}
 {id=WavePlayer_0, properties {soundWaveAssetPtr=&/Launcher_HomingRocket_ReloadInsert_01}}
]
properties {
 attenuationSettings=&Fort_Weapon_Reload_Att
 bOverrideConcurrency=true
 maxDistance=2250
 soundClassObject=&Fort_WeaponFoley
 volumeMultiplier=0.376
 attenuationOverrides {falloffDistance=5000, attenuationShapeExtents {x=1000}}
 concurrencyOverrides {maxCount=1, resolutionRule=StopFarthestThenPreventNew}
}
