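# Sound cue for the start of the homing rocket launcher's reload.
# A single wave player feeds one reload-start wave straight to the output at reduced volume, classed as weapon foley.
# It uses shared weapon reload attenuation and overrides concurrency to allow one instance, stopping the farthest before blocking new ones.
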SoundCue /ResolveGameplay/Items/Guns/HomingRocket/Sounds/Launcher_HomingRocket_ReloadStart_Cue
root=WavePlayer_8
nodes [{id=WavePlayer_8, properties {soundWaveAssetPtr=&/Launcher_HomingRocket_ReloadStart_01}}]
properties {
 attenuationSettings=&Fort_Weapon_Reload_Att
 bOverrideConcurrency=true
 maxDistance=2250
 soundClassObject=&Fort_WeaponFoley
 volumeMultiplier=0.251
 attenuationOverrides {falloffDistance=5000, attenuationShapeExtents {x=1000}}
 concurrencyOverrides {maxCount=1, resolutionRule=StopFarthestThenPreventNew}
}
