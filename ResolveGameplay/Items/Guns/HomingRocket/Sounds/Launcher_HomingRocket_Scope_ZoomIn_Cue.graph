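# Sound cue for the homing rocket launcher's scope zoom-in, playing a single ADS-start wave.
# A Modulator over the Wave Player adds slight random pitch variation and leaves volume steady.
# A concurrency override allows one instance at a time, stopping the oldest.
# It routes to a third-person ranged weapon fire sound class, and its stored attenuation values disable spatialization and reverb send.
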SoundCue /ResolveGameplay/Items/Guns/HomingRocket/Sounds/Launcher_HomingRocket_Scope_ZoomIn_Cue
root=Modulator_2
nodes [
 {id=Modulator_2, children [WavePlayer_3], properties {pitchMax=1.02, pitchMin=0.98, volumeMax=1, volumeMin=1}}
 {id=WavePlayer_3, properties {soundWaveAssetPtr=&/Launcher_HomingRocket_ADS_Start_01}}
]
properties {
 bOverrideConcurrency=true
 maxDistance=2097152
 soundClassObject=&Fort_WeaponRanged_Fire_3P_PSM
 volumeMultiplier=1
 attenuationOverrides {bEnableReverbSend=false, bSpatialize=false, falloffDistance=1500, nonSpatializedRadiusStart=200, attenuationShapeExtents {x=1200}}
 concurrencyOverrides {maxCount=1, resolutionRule=StopOldest}
}
